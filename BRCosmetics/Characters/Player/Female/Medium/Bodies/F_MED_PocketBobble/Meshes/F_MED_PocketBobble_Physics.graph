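# Physics asset for the F_MED_PocketBobble female character body, covering the pelvis, legs and a skirt.
# Eleven three-body skirt chains (deform, dyn, dyn_1) hang from the pelvis with limited swing and twist, alongside leg and foot capsules.
# Cross-constraints tie neighbouring chain tips around the sides and back, leaving the front centre chain unlinked.
# A huge root sphere sits just below the origin like a floor, with the pelvis rotationally free against it.
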PhysicsAsset /BRCosmetics/Characters/Player/Female/Medium/Bodies/F_MED_PocketBobble/Meshes/F_MED_PocketBobble_Physics
bodies {
 calf_l [{shapeType=Capsule, length=44.1626, radius=8.6159, center=<-22.4937 -1.7424 0>, rotation=<rot 90 0 0>}]
 calf_r [{shapeType=Capsule, length=44.1626, radius=8.6159, center=<22.4937 1.7424 0>, rotation=<rot 90 180 0>}]
 deform_skirt_b_l [{shapeType=Sphere, radius=3, shapeName=deform_skirt_b_l_sphere}]
 deform_skirt_b_r [{shapeType=Sphere, radius=3, shapeName=deform_skirt_b_r_sphere}]
 deform_skirt_bk_l [{shapeType=Sphere, radius=3, shapeName=deform_skirt_bk_l_sphere}]
 deform_skirt_bk_r [{shapeType=Sphere, radius=3, shapeName=deform_skirt_bk_r_sphere}]
 deform_skirt_f [{shapeType=Sphere, radius=3, shapeName=deform_skirt_f_sphere}]
 deform_skirt_f_l [{shapeType=Sphere, radius=3, shapeName=deform_skirt_f_l_sphere}]
 deform_skirt_f_m_l [{shapeType=Sphere, radius=3, shapeName=deform_skirt_f_m_l_sphere}]
 deform_skirt_f_m_r [{shapeType=Sphere, radius=3, shapeName=deform_skirt_f_m_r_sphere}]
 deform_skirt_f_r [{shapeType=Sphere, radius=3, shapeName=deform_skirt_f_r_sphere}]
 deform_skirt_m_l [{shapeType=Sphere, radius=3, shapeName=deform_skirt_m_l_sphere}]
 deform_skirt_m_r [{shapeType=Sphere, radius=3, shapeName=deform_skirt_m_r_sphere}]
 dyn_skirt_b_1_l [{shapeType=Capsule, length=10, radius=6, shapeName=dyn_skirt_b_1_l_capsule, center=<7.7943 -2.9302 0>}]
 dyn_skirt_b_1_r [{shapeType=Capsule, length=10, radius=6, shapeName=dyn_skirt_b_1_r_capsule, center=<7.7943 -2.9302 0>}]
 dyn_skirt_b_l [{shapeType=Capsule, length=10, radius=6, shapeName=dyn_skirt_b_l_capsule, center=<6.414 -6.9489 0>}]
 dyn_skirt_b_r [{shapeType=Capsule, length=10, radius=6, shapeName=dyn_skirt_b_r_capsule, center=<6.414 -6.9489 0>}]
 dyn_skirt_bk_1_l [{shapeType=Capsule, length=10, radius=6, shapeName=dyn_skirt_bk_1_l_capsule, center=<7.7943 -2.9302 0>}]
 dyn_skirt_bk_1_r [{shapeType=Capsule, length=10, radius=6, shapeName=dyn_skirt_bk_1_r_capsule, center=<7.7943 -2.9302 0>}]
 dyn_skirt_bk_l [{shapeType=Capsule, length=10, radius=6, shapeName=dyn_skirt_bk_l_capsule, center=<6.414 -6.9489 0>}]
 dyn_skirt_bk_r [{shapeType=Capsule, length=10, radius=6, shapeName=dyn_skirt_bk_r_capsule, center=<6.414 -6.9489 0>}]
 dyn_skirt_f [{shapeType=Sphere, radius=3, shapeName=dyn_skirt_f_sphere}]
 dyn_skirt_f_1 [{shapeType=Sphere, radius=3, shapeName=dyn_skirt_f_1_sphere}]
 dyn_skirt_f_1_l [{shapeType=Sphere, radius=3, shapeName=dyn_skirt_f_1_l_sphere}]
 dyn_skirt_f_1_r [{shapeType=Sphere, radius=3, shapeName=dyn_skirt_f_1_r_sphere}]
 dyn_skirt_f_l [{shapeType=Sphere, radius=3, shapeName=dyn_skirt_f_l_sphere}]
 dyn_skirt_f_m_1_l [{shapeType=Capsule, length=13.75, radius=6, shapeName=dyn_skirt_f_m_1_l_capsule, center=<7.7943 -2.9302 0>}]
 dyn_skirt_f_m_1_r [{shapeType=Capsule, length=13.75, radius=6, shapeName=dyn_skirt_f_m_1_r_capsule, center=<7.7943 -2.9302 0>}]
 dyn_skirt_f_m_l [{shapeType=Capsule, length=10, radius=6, shapeName=dyn_skirt_f_m_l_capsule, center=<6.414 -6.9489 0>}]
 dyn_skirt_f_m_r [{shapeType=Capsule, length=10, radius=6, shapeName=dyn_skirt_f_m_r_capsule, center=<6.414 -6.9489 0>}]
 dyn_skirt_f_r [{shapeType=Sphere, radius=3, shapeName=dyn_skirt_f_r_sphere}]
 dyn_skirt_m_1_l [{shapeType=Capsule, length=10, radius=6, shapeName=dyn_skirt_m_1_l_capsule, center=<7.7943 -2.9302 0>}]
 dyn_skirt_m_1_r [{shapeType=Capsule, length=10, radius=6, shapeName=dyn_skirt_m_1_r_capsule, center=<7.7943 -2.9302 0>}]
 dyn_skirt_m_l [{shapeType=Capsule, length=10, radius=6, shapeName=dyn_skirt_m_l_capsule, center=<6.414 -6.9489 0>}]
 dyn_skirt_m_r [{shapeType=Capsule, length=10, radius=6, shapeName=dyn_skirt_m_r_capsule, center=<6.414 -6.9489 0>}]
 foot_l [{shapeType=Capsule, length=17.3349, radius=9.7391, center=<-3.6846 6.5609 -0.6967>, rotation=<rot -0.8637 3.3562 -83.5398>}]
 foot_r [{shapeType=Capsule, length=17.3349, radius=9.7391, center=<3.6846 -6.5609 0.6967>, rotation=<rot -0.8637 3.3562 96.4602>}]
 pelvis [{shapeType=Capsule, length=11.932, radius=8.4598}]
 root [{shapeType=Sphere, radius=500, shapeName=root_sphere, center=<0 0 -500.648>}]
 thigh_l [{shapeType=Capsule, length=38.6447, radius=7.9115, center=<-21.8545 -0.7044 0.8892>, rotation=<rot 90 0 0>}]
 thigh_r [{shapeType=Capsule, length=38.6447, radius=7.9115, center=<21.8545 0.7044 -0.8892>, rotation=<rot 90 180 0>}]
}
constraints [
 {bone1Name=thigh_l, bone2Name=pelvis, swing1LimitDegrees=45, swing2LimitDegrees=45, twistLimitDegrees=45}
 {bone1Name=calf_l, bone2Name=thigh_l, swing1LimitDegrees=45, swing2LimitDegrees=45, twistLimitDegrees=45}
 {bone1Name=foot_l, bone2Name=calf_l, swing1LimitDegrees=45, swing2LimitDegrees=45, twistLimitDegrees=45}
 {bone1Name=thigh_r, bone2Name=pelvis, swing1LimitDegrees=45, swing2LimitDegrees=45, twistLimitDegrees=45}
 {bone1Name=calf_r, bone2Name=thigh_r, swing1LimitDegrees=45, swing2LimitDegrees=45, twistLimitDegrees=45}
 {bone1Name=foot_r, bone2Name=calf_r, swing1LimitDegrees=45, swing2LimitDegrees=45, twistLimitDegrees=45}
 {bone1Name=deform_skirt_b_l, bone2Name=pelvis, swing1LimitDegrees=35, swing2LimitDegrees=10, twistLimitDegrees=5}
 {bone1Name=dyn_skirt_b_l, bone2Name=deform_skirt_b_l, swing1LimitDegrees=35, swing2LimitDegrees=10, twistLimitDegrees=5}
 {bone1Name=dyn_skirt_b_1_l, bone2Name=dyn_skirt_b_l, swing1LimitDegrees=45, swing2LimitDegrees=15, twistLimitDegrees=10}
 {bone1Name=deform_skirt_b_r, bone2Name=pelvis, swing1LimitDegrees=35, swing2LimitDegrees=10, twistLimitDegrees=5}
 {bone1Name=dyn_skirt_b_r, bone2Name=deform_skirt_b_r, swing1LimitDegrees=35, swing2LimitDegrees=10, twistLimitDegrees=5}
 {bone1Name=dyn_skirt_b_1_r, bone2Name=dyn_skirt_b_r, swing1LimitDegrees=45, swing2LimitDegrees=15, twistLimitDegrees=10}
 {bone1Name=deform_skirt_bk_l, bone2Name=pelvis, swing1LimitDegrees=35, swing2LimitDegrees=10, twistLimitDegrees=5}
 {bone1Name=dyn_skirt_bk_l, bone2Name=deform_skirt_bk_l, swing1LimitDegrees=35, swing2LimitDegrees=10, twistLimitDegrees=1}
 {bone1Name=dyn_skirt_bk_1_l, bone2Name=dyn_skirt_bk_l, swing1LimitDegrees=45, swing2LimitDegrees=15, twistLimitDegrees=10}
 {bone1Name=deform_skirt_bk_r, bone2Name=pelvis, swing1LimitDegrees=35, swing2LimitDegrees=10, twistLimitDegrees=5}
 {bone1Name=dyn_skirt_bk_r, bone2Name=deform_skirt_bk_r, swing1LimitDegrees=35, swing2LimitDegrees=10, twistLimitDegrees=1}
 {bone1Name=dyn_skirt_bk_1_r, bone2Name=dyn_skirt_bk_r, swing1LimitDegrees=45, swing2LimitDegrees=15, twistLimitDegrees=10}
 {bone1Name=deform_skirt_f, bone2Name=pelvis, swing1LimitDegrees=35, swing2LimitDegrees=10, twistLimitDegrees=5}
 {bone1Name=dyn_skirt_f, bone2Name=deform_skirt_f, swing1LimitDegrees=35, swing2LimitDegrees=10, twistLimitDegrees=5}
 {bone1Name=dyn_skirt_f_1, bone2Name=dyn_skirt_f, swing1LimitDegrees=35, swing2LimitDegrees=10, twistLimitDegrees=5}
 {bone1Name=deform_skirt_f_l, bone2Name=pelvis, swing1LimitDegrees=35, swing2LimitDegrees=10, twistLimitDegrees=5}
 {bone1Name=dyn_skirt_f_l, bone2Name=deform_skirt_f_l, swing1LimitDegrees=35, swing2LimitDegrees=10, twistLimitDegrees=5}
 {bone1Name=dyn_skirt_f_1_l, bone2Name=dyn_skirt_f_l, swing1LimitDegrees=35, swing2LimitDegrees=10, twistLimitDegrees=5}
 {bone1Name=deform_skirt_f_m_l, bone2Name=pelvis, swing1LimitDegrees=35, swing2LimitDegrees=10, twistLimitDegrees=5}
 {bone1Name=dyn_skirt_f_m_l, bone2Name=deform_skirt_f_m_l, swing1LimitDegrees=35, swing2LimitDegrees=10, twistLimitDegrees=5}
 {bone1Name=dyn_skirt_f_m_1_l, bone2Name=dyn_skirt_f_m_l, swing1LimitDegrees=45, swing2LimitDegrees=15, twistLimitDegrees=10}
 {bone1Name=deform_skirt_f_m_r, bone2Name=pelvis, swing1LimitDegrees=35, swing2LimitDegrees=10, twistLimitDegrees=5}
 {bone1Name=dyn_skirt_f_m_r, bone2Name=deform_skirt_f_m_r, swing1LimitDegrees=35, swing2LimitDegrees=10, twistLimitDegrees=5}
 {bone1Name=dyn_skirt_f_m_1_r, bone2Name=dyn_skirt_f_m_r, swing1LimitDegrees=45, swing2LimitDegrees=15, twistLimitDegrees=10}
 {bone1Name=deform_skirt_f_r, bone2Name=pelvis, swing1LimitDegrees=35, swing2LimitDegrees=10, twistLimitDegrees=5}
 {bone1Name=dyn_skirt_f_r, bone2Name=deform_skirt_f_r, swing1LimitDegrees=35, swing2LimitDegrees=10, twistLimitDegrees=5}
 {bone1Name=dyn_skirt_f_1_r, bone2Name=dyn_skirt_f_r, swing1LimitDegrees=35, swing2LimitDegrees=10, twistLimitDegrees=5}
 {bone1Name=deform_skirt_m_l, bone2Name=pelvis, swing1LimitDegrees=35, swing2LimitDegrees=10, twistLimitDegrees=5}
 {bone1Name=dyn_skirt_m_l, bone2Name=deform_skirt_m_l, swing1LimitDegrees=35, swing2LimitDegrees=10, twistLimitDegrees=5}
 {bone1Name=dyn_skirt_m_1_l, bone2Name=dyn_skirt_m_l, swing1LimitDegrees=45, swing2LimitDegrees=15, twistLimitDegrees=10}
 {bone1Name=deform_skirt_m_r, bone2Name=pelvis, swing1LimitDegrees=35, swing2LimitDegrees=10, twistLimitDegrees=5}
 {bone1Name=dyn_skirt_m_r, bone2Name=deform_skirt_m_r, swing1LimitDegrees=35, swing2LimitDegrees=10, twistLimitDegrees=5}
 {bone1Name=dyn_skirt_m_1_r, bone2Name=dyn_skirt_m_r, swing1LimitDegrees=45, swing2LimitDegrees=15, twistLimitDegrees=10}
 {
  bone1Name=dyn_skirt_f_m_1_l
  bone2Name=dyn_skirt_f_1_l
  swing1LimitDegrees=15
  swing2LimitDegrees=15
  swing2Motion=Free
  twistLimitDegrees=10
 }
 {bone1Name=dyn_skirt_m_1_l, bone2Name=dyn_skirt_f_m_1_l, swing1LimitDegrees=15, swing2LimitDegrees=15, twistLimitDegrees=10}
 {bone1Name=dyn_skirt_b_1_l, bone2Name=dyn_skirt_m_1_l, swing1LimitDegrees=15, swing2LimitDegrees=15, swing2Motion=Free, twistLimitDegrees=25}
 {
  bone1Name=dyn_skirt_f_m_1_r
  bone2Name=dyn_skirt_f_1_r
  swing1LimitDegrees=15
  swing2LimitDegrees=15
  swing2Motion=Free
  twistLimitDegrees=10
 }
 {bone1Name=dyn_skirt_m_1_r, bone2Name=dyn_skirt_f_m_1_r, swing1LimitDegrees=15, swing2LimitDegrees=15, twistLimitDegrees=10}
 {bone1Name=dyn_skirt_b_1_r, bone2Name=dyn_skirt_m_1_r, swing1LimitDegrees=15, swing2LimitDegrees=15, swing2Motion=Free, twistLimitDegrees=25}
 {
  bone1Name=dyn_skirt_bk_1_l
  bone2Name=dyn_skirt_bk_1_r
  swing1LimitDegrees=15
  swing2LimitDegrees=15
  swing2Motion=Free
  twistLimitDegrees=25
 }
 {bone1Name=dyn_skirt_b_1_r, bone2Name=dyn_skirt_bk_1_r, swing1LimitDegrees=15, swing2LimitDegrees=15, swing2Motion=Free, twistLimitDegrees=25}
 {bone1Name=dyn_skirt_b_1_l, bone2Name=dyn_skirt_bk_1_l, swing1LimitDegrees=15, swing2LimitDegrees=15, swing2Motion=Free, twistLimitDegrees=25}
 {
  bone1Name=pelvis
  bone2Name=root
  swing1LimitDegrees=45
  swing1Motion=Free
  swing2LimitDegrees=45
  swing2Motion=Free
  twistLimitDegrees=45
  twistMotion=Free
 }
]
