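# Post-process Control Rig for the PocketBobble female medium body, correcting toy, back bow and keychain accessory bones around the skirt and belt.
# Weighted parent constraints blend those bones from spine, skirt, belt, groin and other body bones.
# Aim constraints orient keychains toward skirt or keychain bones, followed by a fixed rotation tweak and rotation blends from skirt bones.
# A spherical pose reader on the left thigh, remapped through a curve, drives corrective offsets on the left toy bones.
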ControlRigBlueprint /BRCosmetics/Characters/Player/Female/Medium/Bodies/F_MED_PocketBobble/Meshes/F_MED_PocketBobble_Post_CtrlRig
graphs [
 RigVMFunctionLibrary {links [], nodes []}
 RigVMModel {nodes [
   RigUnit_BeginExecution = &RigUnit_BeginExecution()
   &RigUnit_ParentConstraint(
    AdvancedSettings=(RotationOrderForFilter=XZY)
    Child=(Type=Bone,Name="deform_toy_2_r")
    Parents="((Item=(Type=Bone,Name=\"spine_03\"),Weight=0.0194),(Item=(Type=Bone,Name=\"deform_groin_r\"),Weight=0.0222),(Item=(Type=Bone,Name=\"dyn_belt_r\"),Weight=0.0279),(Item=(Type=Bone,Name=\"dyn_skirt_f_m_r\"),Weight=0.0458),(Item=(Type=Bone,Name=\"dyn_skirt_m_r\"),Weigh ...(truncated)"
    Weight=1
    ExecutePin=@RigUnit_BeginExecution.ExecutePin
   )
   > &RigUnit_ParentConstraint(
    AdvancedSettings=(RotationOrderForFilter=XZY)
    Child=(Type=Bone,Name="deform_toy_1_r")
    Parents="((Item=(Type=Bone,Name=\"spine_03\"),Weight=0.0093),(Item=(Type=Bone,Name=\"dyn_belt_f\"),Weight=0.015),(Item=(Type=Bone,Name=\"dyn_belt_r\"),Weight=0.0332),(Item=(Type=Bone,Name=\"dyn_skirt_f_m_r\"),Weight=0.0433),(Item=(Type=Bone,Name=\"spine_04\"),Weight=0.0464), ...(truncated)"
    Weight=1
   )
   > &RigUnit_ParentConstraint(
    AdvancedSettings=(RotationOrderForFilter=XZY)
    Child=(Type=Bone,Name="deform_toy_l")
    Parents="((Item=(Type=Bone,Name=\"spine_01\"),Weight=0.0181),(Item=(Type=Bone,Name=\"dyn_belt_f\"),Weight=0.0206),(Item=(Type=Bone,Name=\"deform_groin_l\"),Weight=0.0261),(Item=(Type=Bone,Name=\"spine_03\"),Weight=0.0375),(Item=(Type=Bone,Name=\"dyn_belt_l\"),Weight=0.0431), ...(truncated)"
    Weight=1
   )
   > &RigUnit_ParentConstraint(
    AdvancedSettings=(RotationOrderForFilter=XZY)
    Child=(Type=Bone,Name="deform_toy_bottom_l")
    Parents="((Item=(Type=Bone,Name=\"dyn_skirt_f_m_l\"),Weight=0.008),(Item=(Type=Bone,Name=\"dyn_belt_f\"),Weight=0.0154),(Item=(Type=Bone,Name=\"spine_03\"),Weight=0.0247),(Item=(Type=Bone,Name=\"dyn_belt_l\"),Weight=0.0342),(Item=(Type=Bone,Name=\"deform_groin_l\"),Weight=0. ...(truncated)"
    Weight=1
   )
   > &RigUnit_ParentConstraint(
    AdvancedSettings=(RotationOrderForFilter=XZY)
    Child=(Type=Bone,Name="deform_keychain_bk")
    Parents="((Item=(Type=Bone,Name=\"clavicle_r\"),Weight=0.0009),(Item=(Type=Bone,Name=\"clavicle_l\"),Weight=0.0031),(Item=(Type=Bone,Name=\"spine_05\"),Weight=0.0046),(Item=(Type=Bone,Name=\"spine_03\"),Weight=0.0711),(Item=(Type=Bone,Name=\"spine_04\"),Weight=0.0844),(Item= ...(truncated)"
    Weight=1
    bMaintainOffset=true
   )
   > &RigUnit_ParentConstraint(
    AdvancedSettings=(RotationOrderForFilter=XZY)
    Child=(Type=Bone,Name="deform_bow_bk")
    Parents="((Item=(Type=Bone,Name=\"clavicle_l\"),Weight=0.0007),(Item=(Type=Bone,Name=\"clavicle_r\"),Weight=0.0017),(Item=(Type=Bone,Name=\"spine_05\"),Weight=0.0033),(Item=(Type=Bone,Name=\"spine_03\"),Weight=0.0691),(Item=(Type=Bone,Name=\"spine_04\"),Weight=0.0793),(Item= ...(truncated)"
    Weight=1
   )
   > &RigUnit_AimConstraintLocalSpaceOffset(
    Child=(Type=Bone,Name="deform_keychain_bk")
    Filter=(bX=True,bY=True,bZ=True)
    Parents=((Item=(Type=Bone,Name="dyn_skirt_bk_1_l"),Weight=1))
    Weight=1
    WorldUp=(Target=(X=0,Y=0,Z=1),Kind=Location,Space=(Type=Bone,Name="deform_keychain_bk"))
    bMaintainOffset=true
   )
   > &RigUnit_ParentConstraint(
    AdvancedSettings=(RotationOrderForFilter=XZY)
    Child=(Type=Bone,Name="deform_keychain_f")
    Parents="((Item=(Type=Bone,Name=\"dyn_belt_l\"),Weight=0.0195),(Item=(Type=Bone,Name=\"deform_groin_l\"),Weight=0.0237),(Item=(Type=Bone,Name=\"spine_03\"),Weight=0.0246),(Item=(Type=Bone,Name=\"deform_groin_r\"),Weight=0.0305),(Item=(Type=Bone,Name=\"spine_04\"),Weight=0.06 ...(truncated)"
    Weight=1
   )
   > &RigUnit_ParentConstraint(
    AdvancedSettings=(RotationOrderForFilter=XZY)
    Child=(Type=Bone,Name="deform_keychain_f_r")
    Parents="((Item=(Type=Bone,Name=\"deform_groin_l\"),Weight=0.0107),(Item=(Type=Bone,Name=\"spine_03\"),Weight=0.0198),(Item=(Type=Bone,Name=\"dyn_belt_r\"),Weight=0.0256),(Item=(Type=Bone,Name=\"dyn_belt_f\"),Weight=0.0533),(Item=(Type=Bone,Name=\"spine_04\"),Weight=0.0551), ...(truncated)"
    Weight=1
   )
   > &RigUnit_ParentConstraint(
    AdvancedSettings=(RotationOrderForFilter=XZY)
    Child=(Type=Bone,Name="deform_keychain_f_l")
    Parents="((Item=(Type=Bone,Name=\"dyn_skirt_f_m_l\"),Weight=0.0076),(Item=(Type=Bone,Name=\"dyn_belt_f\"),Weight=0.021),(Item=(Type=Bone,Name=\"spine_03\"),Weight=0.0215),(Item=(Type=Bone,Name=\"dyn_belt_l\"),Weight=0.034),(Item=(Type=Bone,Name=\"deform_groin_l\"),Weight=0.0 ...(truncated)"
    Weight=1
   )
   > &RigUnit_AimConstraintLocalSpaceOffset(
    Child=(Type=Bone,Name="deform_keychain_f")
    Filter=(bX=true,bY=true,bZ=True)
    Parents=((Item=(Type=Bone,Name="dyn_skirt_f_1"),Weight=1))
    UpAxis=(X=0,Y=1,Z=0)
    Weight=1
    WorldUp=(Target=(X=0,Y=0,Z=1),Kind=Location,Space=(Type=Bone,Name="deform_keychain_f"))
    bMaintainOffset=true
   )
   > &RigUnit_AimConstraintLocalSpaceOffset(
    Child=(Type=Bone,Name="dyn_keychain_f_l")
    Filter=(bX=true,bY=true,bZ=true)
    Parents=((Item=(Type=Bone,Name="dyn_keychain_1_f"),Weight=1))
    UpAxis=(X=0,Y=1,Z=0)
    Weight=1
    WorldUp=(Target=(X=0,Y=0,Z=1),Kind=Location,Space=(Type=Bone,Name="deform_keychain_f_l"))
    bMaintainOffset=true
   )
   > &RigUnit_AimConstraintLocalSpaceOffset(
    Child=(Type=Bone,Name="dyn_keychain_f_r")
    Filter=(bX=true,bY=true,bZ=True)
    Parents=((Item=(Type=Bone,Name="dyn_keychain_1_f"),Weight=1))
    UpAxis=(X=0,Y=1,Z=0)
    Weight=1
    WorldUp=(Target=(X=0,Y=0,Z=1),Kind=Location,Space=(Type=Bone,Name="deform_keychain_f_r"))
    bMaintainOffset=true
   )
   > &RigUnit_ModifyTransforms(
    ItemToModify=((Item=(Type=Bone,Name="deform_keychain_f"),Transform=(Rotation=(X=0,Y=0,Z=-0.0872,W=0.9962),Translation=(X=0,Y=0,Z=0),Scale3D=(X=1,Y=1,Z=1))))
    Weight=1
    WeightMinimum=0
   )
   > &RigUnit_ParentConstraint(
    AdvancedSettings=(InterpolationType=Average,RotationOrderForFilter=XZY)
    Child=(Type=Bone,Name="dyn_keychain_f_l")
    Filter=(TranslationFilter=(bX=false,bY=false,bZ=false),RotationFilter=(bX=true,bY=true,bZ=True),ScaleFilter=(bX=True,bY=True,bZ=True))
    Parents=((Item=(Type=Bone,Name="dyn_skirt_f_l"),Weight=1))
    Weight=1
    bMaintainOffset=true
   )
   > &RigUnit_ParentConstraint(
    AdvancedSettings=(InterpolationType=Average,RotationOrderForFilter=XZY)
    Child=(Type=Bone,Name="dyn_keychain_f_r")
    Filter=(TranslationFilter=(bX=true,bY=true,bZ=true),RotationFilter=(bX=true,bY=true,bZ=True),ScaleFilter=(bX=True,bY=True,bZ=True))
    Parents=((Item=(Type=Bone,Name="dyn_skirt_f_r"),Weight=0.5),(Item=(Type=Bone,Name="dyn_skirt_f"),Weight=0.5))
    Weight=1
    bMaintainOffset=true
   )
   > n1 = &RigUnit_SphericalPoseReader(
    ActiveRegionScaleFactors=(PositiveWidth=0.7,NegativeWidth=0.4,PositiveHeight=0.3,NegativeHeight=0.35)
    ActiveRegionSize=0.3
    Debug=(bDrawDebug=false,bDraw2D=false,DebugScale=25,DebugSegments=20,DebugThickness=0.25)
    DriverAxis=(X=-1,Y=0,Z=0)
    DriverItem=(Type=Bone,Name="thigh_l")
    FalloffRegionScaleFactors=(PositiveWidth=0.5,NegativeWidth=1,PositiveHeight=0.4,NegativeHeight=0.5)
    FalloffSize=0.95
    OptionalParentItem=(Type=Bone,Name="spine_03")
    RotationOffset=(X=90,Y=10,Z=-80)
   )
   RigVMFunction_AnimEvalRichCurve = &RigVMFunction_AnimEvalRichCurve(
    Curve=(EditorCurveData=(Keys=((InterpMode=RCIM_Cubic,TangentMode=RCTM_User),(InterpMode=RCIM_Cubic,TangentMode=RCTM_User,Time=1,Value=1,ArriveTangent=1.0946,LeaveTangent=1.0946))))
    SourceMinimum=0
    TargetMinimum=0
    Value=@n1.OutputParam
   )
   > &RigUnit_ModifyTransforms(
    ItemToModify="((Item=(Type=Bone,Name=\"deform_toy_l\"),Transform=(Rotation=(X=0.2527,Y=0.056,Z=0.2091,W=0.943),Translation=(X=-3,Y=-1,Z=0),Scale3D=(X=1,Y=1,Z=1))),(Item=(Type=Bone,Name=\"deform_toy_bottom_l\"),Transform=(Rotation=(X=0.467,Y=0.1889,Z=-0.551,W=0.6653),Transla ...(truncated)"
    WeightMinimum=0
    Weight=@RigVMFunction_AnimEvalRichCurve.Result
   )
  ]}
]
hierarchy [
 root {loc=<0 0 0>, rot=<0 0 0 1>}
 pelvis {parent=root, loc=<0 -0.5301 98.0571>, rot=<0.0127 0.707 -0.0127 -0.707>}
 spine_01 {parent=pelvis, loc=<6.1208 -0.9328 0>, rot=<0 0 0.034 -0.9994>}
 spine_02 {parent=spine_01, loc=<6.5997 -0.2095 0>, rot=<0 0 0.0238 -0.9997>}
 spine_03 {parent=spine_02, loc=<6.4914 -0.1237 0>, rot=<0 0 0.0149 -0.9999>}
 spine_04 {parent=spine_03, loc=<6.9598 0.0794 0>, rot=<0 0 -0.0132 -0.9999>}
 spine_05 {parent=spine_04, loc=<6.9987 0 0>, rot=<0 0 -0.0015 -1>}
 neck_01 {parent=spine_05, loc=<11.3138 0.1529 0>, rot=<0 0 -0.2003 -0.9797>}
 neck_02 {parent=neck_01, loc=<4.5627 0 0>, rot=<0 0 0.0504 -0.9987>}
 head {parent=neck_02, loc=<5.3914 0 0>, rot=<0 0 0.115 -0.9934>}
 clavicle_l {parent=spine_05, loc=<5.4831 -0.1053 -3.2946>, rot=<0.0465 -0.7289 -0.009 -0.6829>}
 upperarm_l {parent=clavicle_l, loc=<9.8658 -1.2136 -0.4105>, rot=<0.0164 -0.3245 -0.0269 -0.9454>}
 lowerarm_l {parent=upperarm_l, loc=<26.465 -0.8345 0.259>, rot=<0.0036 0.0161 -0.2298 -0.9731>}
 hand_l {parent=lowerarm_l, loc=<22.7355 -0.3203 -0.6105>, rot=<-0.6784 -0.0223 0.0735 -0.7307>}
 thumb_01_l {parent=hand_l, loc=<1.3118 -0.3191 -2.459>, rot=<0.6546 -0.4042 0.0664 -0.6354>}
 thumb_02_l {parent=thumb_01_l, loc=<5.9336 -0.0448 0.0401>, rot=<-0.0037 0.0213 0.298 -0.9543>}
 thumb_03_l {parent=thumb_02_l, loc=<2.2223 0.3682 -0.0907>, rot=<0.0118 -0.013 0.0066 -0.9998>}
 index_metacarpal_l {parent=hand_l, loc=<1.6535 0.0183 -2.96>, rot=<0.0158 -0.0124 -0.0173 -0.9996>}
 index_01_l {parent=index_metacarpal_l, loc=<8.3008 0.0357 -0.0524>, rot=<-0.0044 0.0485 0.0924 -0.9945>}
 index_02_l {parent=index_01_l, loc=<3.3587 -0.3004 0.0452>, rot=<-0.002 0.0027 0.3225 -0.9466>}
 index_03_l {parent=index_02_l, loc=<2.5578 -0.123 -0.0252>, rot=<-0.002 0.0067 0.2558 -0.9667>}
 middle_metacarpal_l {parent=hand_l, loc=<2.1856 -0.1725 -0.5984>, rot=<0.006 0.0168 -0.0329 -0.9993>}
 middle_01_l {parent=middle_metacarpal_l, loc=<7.3275 -0.2447 0.0129>, rot=<-0.0208 0.0042 0.0858 -0.9961>}
 middle_02_l {parent=middle_01_l, loc=<4.0123 0.1401 0.1175>, rot=<-0.0218 0.0258 0.3512 -0.9357>}
 middle_03_l {parent=middle_02_l, loc=<2.6077 0.0534 0.1373>, rot=<0.0082 0.0188 0.1708 -0.9851>}
 ring_metacarpal_l {parent=hand_l, loc=<2.2074 -0.3625 1.3638>, rot=<-0.0284 0.0722 -0.0458 -0.9959>}
 ring_01_l {parent=ring_metacarpal_l, loc=<7.0295 -0.6197 -0.1446>, rot=<-0.0489 -0.0252 0.1892 -0.9804>}
 ring_02_l {parent=ring_01_l, loc=<3.6174 0.4647 -0.1027>, rot=<0 0 0.2806 -0.9598>}
 ring_03_l {parent=ring_02_l, loc=<2.4152 0.114 0.0199>, rot=<0 0 0.1189 -0.9929>}
 pinky_metacarpal_l {parent=hand_l, loc=<1.7616 -0.9753 3.3814>, rot=<-0.0397 0.0814 -0.069 -0.9935>}
 pinky_01_l {parent=pinky_metacarpal_l, loc=<7.1709 -0.4831 -0.1094>, rot=<-0.0654 -0.0035 0.2242 -0.9723>}
 pinky_02_l {parent=pinky_01_l, loc=<2.8274 0.0893 -0.1043>, rot=<0.0037 0.0011 0.2854 -0.9584>}
 pinky_03_l {parent=pinky_02_l, loc=<2.2046 0.0307 -0.0078>, rot=<-0.0002 0.0098 0.1445 -0.9895>}
 deform_wrist_dn_base_l {parent=hand_l, loc=<0 -1.5326 0>, rot=<0 0 0 -1>}
 deform_wrist_dn_l {parent=deform_wrist_dn_base_l, loc=<0 0 0>, rot=<0 0 0 1>}
 deform_wrist_up_base_l {parent=hand_l, loc=<0 2.3009 0>, rot=<0 0 0 -1>}
 deform_wrist_up_l {parent=deform_wrist_up_base_l, loc=<0 0 0>, rot=<0 0 0 1>}
 lowerarm_twist_02_l {parent=lowerarm_l, loc=<10.6707 0 0>, rot=<0 0 0 1>}
 lowerarm_twist_01_l {parent=lowerarm_l, loc=<18.316 0 0>, rot=<0 0 0 1>}
 deform_elbow_in_base_l {parent=lowerarm_l, loc=<0.1834 3.6062 -0.2572>, rot=<0 0 0 -1>}
 deform_elbow_in_l {parent=deform_elbow_in_base_l, loc=<0 0 0>, rot=<0 0 0 1>}
 deform_elbow_out_base_l {parent=lowerarm_l, loc=<0 -4 0>, rot=<0 0 0 1>}
 deform_elbow_out_l {parent=deform_elbow_out_base_l, loc=<0 0 0>, rot=<0 0 0 1>}
 upperarm_twist_01_l {parent=upperarm_l, loc=<7.2135 -0.1134 0>, rot=<0 0 0 1>}
 deform_upperarm_bk_base_l {parent=upperarm_twist_01_l, loc=<-6.4053 -4.3746 0.314>, rot=<0 0 0 -1>}
 deform_upperarm_bk_l {parent=deform_upperarm_bk_base_l, loc=<0 0 0>, rot=<0 0 0 1>}
 deform_upperarm_fr_base_l {parent=upperarm_twist_01_l, loc=<-5.6312 5.092 0.3354>, rot=<0 0 0 -1>}
 deform_upperarm_fr_l {parent=deform_upperarm_fr_base_l, loc=<0 0 0>, rot=<0 0 0 1>}
 deform_upperarm_up_base_l {parent=upperarm_twist_01_l, loc=<-8.891 0.4825 3.9517>, rot=<0 0 0 -1>}
 deform_upperarm_up_l {parent=deform_upperarm_up_base_l, loc=<0 0 0>, rot=<0 0 0 1>}
 upperarm_twist_02_l {parent=upperarm_l, loc=<19.7998 -0.1134 0>, rot=<0 0 0 1>}
 clavicle_r {parent=spine_05, loc=<5.4831 -0.1053 3.2946>, rot=<0.7289 0.0465 -0.6829 0.009>}
 upperarm_r {parent=clavicle_r, loc=<-9.8658 1.2136 0.4105>, rot=<0.0164 -0.3245 -0.0269 -0.9454>}
 lowerarm_r {parent=upperarm_r, loc=<-26.465 0.8345 -0.259>, rot=<0.0036 0.0161 -0.2298 -0.9731>}
 hand_r {parent=lowerarm_r, loc=<-22.7355 0.3203 0.6106>, rot=<-0.6784 -0.0223 0.0735 -0.7307>}
 thumb_01_r {parent=hand_r, loc=<-1.3118 0.3191 2.459>, rot=<0.6546 -0.4042 0.0664 -0.6354>}
 thumb_02_r {parent=thumb_01_r, loc=<-5.9336 0.0448 -0.0401>, rot=<-0.0037 0.0213 0.298 -0.9543>}
 thumb_03_r {parent=thumb_02_r, loc=<-2.2223 -0.3682 0.0907>, rot=<0.0118 -0.013 0.0066 -0.9998>}
 index_metacarpal_r {parent=hand_r, loc=<-1.6535 -0.0183 2.96>, rot=<0.0158 -0.0124 -0.0173 -0.9996>}
 index_01_r {parent=index_metacarpal_r, loc=<-8.3008 -0.0357 0.0524>, rot=<-0.0044 0.0485 0.0924 -0.9945>}
 index_02_r {parent=index_01_r, loc=<-3.3587 0.3004 -0.0452>, rot=<-0.002 0.0027 0.3225 -0.9466>}
 index_03_r {parent=index_02_r, loc=<-2.5578 0.123 0.0252>, rot=<-0.002 0.0067 0.2558 -0.9667>}
 middle_metacarpal_r {parent=hand_r, loc=<-2.1856 0.1725 0.5984>, rot=<0.006 0.0168 -0.0329 -0.9993>}
 middle_01_r {parent=middle_metacarpal_r, loc=<-7.3275 0.2447 -0.0129>, rot=<-0.0208 0.0042 0.0858 -0.9961>}
 middle_02_r {parent=middle_01_r, loc=<-4.0123 -0.1401 -0.1175>, rot=<-0.0218 0.0258 0.3512 -0.9357>}
 middle_03_r {parent=middle_02_r, loc=<-2.6077 -0.0534 -0.1373>, rot=<0.0082 0.0188 0.1708 -0.9851>}
 ring_metacarpal_r {parent=hand_r, loc=<-2.2074 0.3625 -1.3638>, rot=<-0.0284 0.0722 -0.0458 -0.9959>}
 ring_01_r {parent=ring_metacarpal_r, loc=<-7.0295 0.6197 0.1446>, rot=<-0.0489 -0.0252 0.1892 -0.9804>}
 ring_02_r {parent=ring_01_r, loc=<-3.6174 -0.4647 0.1027>, rot=<0 0 0.2806 -0.9598>}
 ring_03_r {parent=ring_02_r, loc=<-2.4152 -0.114 -0.0199>, rot=<0 0 0.1189 -0.9929>}
 pinky_metacarpal_r {parent=hand_r, loc=<-1.7616 0.9753 -3.3814>, rot=<-0.0397 0.0814 -0.069 -0.9935>}
 pinky_01_r {parent=pinky_metacarpal_r, loc=<-7.1709 0.4831 0.1094>, rot=<-0.0654 -0.0035 0.2242 -0.9723>}
 pinky_02_r {parent=pinky_01_r, loc=<-2.8274 -0.0893 0.1043>, rot=<0.0037 0.0011 0.2854 -0.9584>}
 pinky_03_r {parent=pinky_02_r, loc=<-2.2046 -0.0307 0.0078>, rot=<-0.0002 0.0098 0.1445 -0.9895>}
 deform_wrist_dn_base_r {parent=hand_r, loc=<-0.0001 1.5328 0>, rot=<0 0 0 -1>}
 deform_wrist_dn_r {parent=deform_wrist_dn_base_r, loc=<0 0 0>, rot=<0 0 0 1>}
 deform_wrist_up_base_r {parent=hand_r, loc=<-0.0002 -2.3007 -0.0001>, rot=<0 0 0 -1>}
 deform_wrist_up_r {parent=deform_wrist_up_base_r, loc=<0 0 0>, rot=<0 0 0 1>}
 lowerarm_twist_02_r {parent=lowerarm_r, loc=<-10.6707 0 0>, rot=<0 0 0 -1>}
 lowerarm_twist_01_r {parent=lowerarm_r, loc=<-18.316 0 0>, rot=<0 0 0 -1>}
 deform_elbow_in_base_r {parent=lowerarm_r, loc=<-0.1834 -3.6062 0.2572>, rot=<0 0 0 -1>}
 deform_elbow_in_r {parent=deform_elbow_in_base_r, loc=<0 0 0>, rot=<0 0 0 1>}
 deform_elbow_out_base_r {parent=lowerarm_r, loc=<0 4 -0.0001>, rot=<0 0 0 -1>}
 deform_elbow_out_r {parent=deform_elbow_out_base_r, loc=<0 0 0>, rot=<0 0 0 1>}
 upperarm_twist_01_r {parent=upperarm_r, loc=<-7.2135 0.1134 0>, rot=<0 0 0 -1>}
 deform_upperarm_bk_base_r {parent=upperarm_twist_01_r, loc=<6.4053 4.3746 -0.314>, rot=<0 0 0 -1>}
 deform_upperarm_bk_r {parent=deform_upperarm_bk_base_r, loc=<0 0 0>, rot=<0 0 0 1>}
 deform_upperarm_fr_base_r {parent=upperarm_twist_01_r, loc=<5.6312 -5.092 -0.3354>, rot=<0 0 0 -1>}
 deform_upperarm_fr_r {parent=deform_upperarm_fr_base_r, loc=<0 0 0>, rot=<0 0 0 1>}
 deform_upperarm_up_base_r {parent=upperarm_twist_01_r, loc=<8.891 -0.4825 -3.9517>, rot=<0 0 0 -1>}
 deform_upperarm_up_r {parent=deform_upperarm_up_base_r, loc=<0 0 0>, rot=<0 0 0 1>}
 upperarm_twist_02_r {parent=upperarm_r, loc=<-19.7998 0.1134 0>, rot=<0 0 0 -1>}
 attach_backpack {parent=spine_05, loc=<-0.123 -13.473 0>, rot=<-0.0093 -0.707 0.0093 -0.707>}
 Backpack_BR {parent=attach_backpack, type=Null, loc=<0 0 0>, rot=<0 0 0 1>}
 attach_cape {parent=spine_05, loc=<9.017 -11.714 0>, rot=<0.0616 -0.7044 -0.0616 -0.7044>}
 cape {parent=attach_cape, type=Null, loc=<0 0 0>, rot=<0 0 0 1>}
 deform_pec_aim_base_l {parent=spine_04, loc=<3.2456 9.3309 -1.2237>, rot=<0.0603 -0.4033 0.3318 -0.8507>}
 deform_pec_aim_l {parent=deform_pec_aim_base_l, loc=<0 0 0>, rot=<0 0 0 1>}
 deform_pec_base_l {parent=deform_pec_aim_l, loc=<9.3665 0 0>, rot=<0 0 0 1>}
 deform_pec_l {parent=deform_pec_base_l, loc=<0 0 0>, rot=<0 0 0 1>}
 deform_pec_aim_base_r {parent=spine_04, loc=<3.2452 9.3309 1.2237>, rot=<-0.4033 -0.0603 0.8507 0.3318>}
 deform_pec_aim_r {parent=deform_pec_aim_base_r, loc=<0 0 0>, rot=<0 0 0 1>}
 deform_pec_base_r {parent=deform_pec_aim_r, loc=<-9.3667 -0.0003 -0.0004>, rot=<0 0 0 1>}
 deform_pec_r {parent=deform_pec_base_r, loc=<0 0 0>, rot=<0 0 0 1>}
 deform_toy_2_r {parent=spine_01, loc=<-0.5027 2.3139 14.0505>, rot=<0.1396 -0.5423 0.822 0.1037>}
 dyn_toy_2_r {parent=deform_toy_2_r, loc=<0 0 0>, rot=<0 0 0 1>}
 deform_toy_1_r {parent=spine_01, loc=<-0.2862 7.0637 9.2047>, rot=<0.1299 -0.2277 0.961 0.0877>}
 dyn_toy_1_r {parent=deform_toy_1_r, loc=<0 0 0>, rot=<0 0 0 1>}
 deform_toy_l {parent=spine_01, loc=<3.6154 6.0503 -10.5197>, rot=<-0.1474 0.3944 0.9007 0.107>}
 dyn_skirt_toy_l {parent=deform_toy_l, loc=<0 0 0>, rot=<0 0 0 1>}
 dyn_skirt_toy_1_l {parent=dyn_skirt_toy_l, loc=<4.0597 -1.5566 0>, rot=<0 0 0 1>}
 deform_bow_bk {parent=spine_01, loc=<4.9093 -7.1118 2.9381>, rot=<0.2204 -0.9742 0.0411 0.0245>}
 dyn_bow_bk {parent=deform_bow_bk, loc=<0 0 0>, rot=<0 0 -0.0073 -1>}
 dyn_bow_bk_4 {parent=dyn_bow_bk, loc=<8.4312 -0.2597 2.3051>, rot=<0 0.1372 0 -0.9905>}
 dyn_bow_bk_3 {parent=dyn_bow_bk, loc=<8.4919 -0.5115 -2.0937>, rot=<0 -0.0973 0 -0.9953>}
 dyn_bow_bk_1 {parent=deform_bow_bk, loc=<-0.0007 0.0455 -0.9505>, rot=<0.2775 -0.6523 0.0423 -0.704>}
 dyn_bow_bk_2 {parent=deform_bow_bk, loc=<0.0398 -0.1047 1.0199>, rot=<-0.2157 0.6895 -0.0068 -0.6914>}
 deform_skirt_f_r {parent=spine_01, loc=<-0.7823 7.6426 6.8493>, rot=<0.0652 -0.0639 0.9931 0.073>}
 dyn_skirt_f_r {parent=deform_skirt_f_r, loc=<0 0 0>, rot=<0 0 0 1>}
 dyn_skirt_f_1_r {parent=dyn_skirt_f_r, loc=<12.4534 0 0>, rot=<0 0 0 1>}
 deform_skirt_f_m_r {parent=spine_01, loc=<-0.6941 4.9101 10.4185>, rot=<0.1474 -0.3944 0.9007 0.107>}
 dyn_skirt_f_m_r {parent=deform_skirt_f_m_r, loc=<0 0 0>, rot=<0 0 0 1>}
 dyn_skirt_f_m_1_r {parent=dyn_skirt_f_m_r, loc=<12.4534 0 0>, rot=<0 0 0 1>}
 deform_skirt_m_r {parent=spine_01, loc=<-0.5399 0.1303 11.5359>, rot=<0.161 -0.6933 0.6902 0.1304>}
 dyn_skirt_m_r {parent=deform_skirt_m_r, loc=<0 0 0>, rot=<0 0 0 -1>}
 dyn_skirt_m_1_r {parent=dyn_skirt_m_r, loc=<12.437 -0.6375 0>, rot=<0 0 0.0256 -0.9997>}
 deform_skirt_b_r {parent=spine_01, loc=<-0.3929 -4.4266 9.675>, rot=<0.2388 -0.8782 0.3845 0.1545>}
 dyn_skirt_b_r {parent=deform_skirt_b_r, loc=<0 0 0>, rot=<0 0 0 1>}
 dyn_skirt_b_1_r {parent=dyn_skirt_b_r, loc=<12.4534 0 0>, rot=<0 0 -0.1433 -0.9897>}
 deform_skirt_bk_r {parent=spine_01, loc=<-0.3144 -6.8567 5.0656>, rot=<0.2203 -0.9734 0.0454 0.0438>}
 dyn_skirt_bk_r {parent=deform_skirt_bk_r, loc=<0 0 0>, rot=<0 0 -0.0073 -1>}
 dyn_skirt_bk_1_r {parent=dyn_skirt_bk_r, loc=<12.4535 0 0>, rot=<0 0 -0.1289 -0.9917>}
 deform_skirt_f {parent=spine_01, loc=<-0.7963 8.075 0>, rot=<-0.0003 -0.0036 0.9976 0.0686>}
 dyn_skirt_f {parent=deform_skirt_f, loc=<0 0 0>, rot=<0 0 0 1>}
 dyn_skirt_f_1 {parent=dyn_skirt_f, loc=<12.4533 0 0>, rot=<0 0 0 1>}
 deform_skirt_bk_l {parent=spine_01, loc=<-0.3145 -6.8567 -5.0656>, rot=<0.2203 -0.9734 -0.0454 -0.0438>}
 dyn_skirt_bk_l {parent=deform_skirt_bk_l, loc=<0 0 0>, rot=<0 0 -0.0073 -1>}
 dyn_skirt_bk_1_l {parent=dyn_skirt_bk_l, loc=<12.4533 0 0>, rot=<0 0 -0.1289 -0.9917>}
 deform_skirt_b_l {parent=spine_01, loc=<-0.3929 -4.4266 -9.675>, rot=<0.2388 -0.8782 -0.3845 -0.1545>}
 dyn_skirt_b_l {parent=deform_skirt_b_l, loc=<0 0 0>, rot=<0 0 0 1>}
 dyn_skirt_b_1_l {parent=dyn_skirt_b_l, loc=<12.4533 0 0>, rot=<0 0 -0.1433 -0.9897>}
 deform_skirt_m_l {parent=spine_01, loc=<-0.54 0.1303 -11.5359>, rot=<0.161 -0.6933 -0.6902 -0.1304>}
 dyn_skirt_m_l {parent=deform_skirt_m_l, loc=<0 0 0>, rot=<0 0 0 1>}
 dyn_skirt_m_1_l {parent=dyn_skirt_m_l, loc=<12.437 -0.6375 0>, rot=<0 0 0.0256 -0.9997>}
 deform_skirt_f_m_l {parent=spine_01, loc=<-0.6942 4.9101 -10.4185>, rot=<-0.1474 0.3944 0.9007 0.107>}
 dyn_skirt_f_m_l {parent=deform_skirt_f_m_l, loc=<0 0 0>, rot=<0 0 0 1>}
 dyn_skirt_f_m_1_l {parent=dyn_skirt_f_m_l, loc=<12.4533 0 0>, rot=<0 0 0 1>}
 deform_skirt_f_l {parent=spine_01, loc=<-0.7823 7.6426 -6.8493>, rot=<-0.0652 0.0639 0.9931 0.073>}
 dyn_skirt_f_l {parent=deform_skirt_f_l, loc=<0 0 0>, rot=<0 0 0 1>}
 dyn_skirt_f_1_l {parent=dyn_skirt_f_l, loc=<12.4533 0 0>, rot=<0 0 0 1>}
 deform_belt_l {parent=spine_01, loc=<-0.1774 7.5016 -8.4386>, rot=<-0.0355 -0.0003 0.9992 -0.0185>}
 dyn_belt_l {parent=deform_belt_l, loc=<0 0 0>, rot=<0 0 0 1>}
 deform_belt_r {parent=spine_01, loc=<-0.1774 7.5016 8.4386>, rot=<0.0355 0.0003 0.9992 -0.0185>}
 dyn_belt_r {parent=deform_belt_r, loc=<0 0 0>, rot=<0 0 0 1>}
 deform_belt_f {parent=spine_01, loc=<-0.6789 8.8878 0.0467>, rot=<-0.0089 0.0001 0.9998 -0.0185>}
 dyn_belt_f {parent=deform_belt_f, loc=<0 0 0>, rot=<0 0 0 1>}
 deform_toy_bottom_l {parent=spine_01, loc=<1.5903 6.0949 -11.1393>, rot=<-0.1009 0.4148 0.9005 0.0833>}
 dyn_skirt_toy_bottom_l {parent=deform_toy_bottom_l, loc=<0 0 0>, rot=<0 0 0 1>}
 deform_keychain_bk {parent=spine_01, loc=<2.8306 -7.5286 -4.1768>, rot=<0.2248 -0.9538 -0.1433 -0.1384>}
 dyn_keychain_bk {parent=deform_keychain_bk, loc=<0 0 0>, rot=<0 0 -0.0073 -1>}
 deform_keychain_f {parent=spine_01, loc=<1.2943 9.2106 -1.8314>, rot=<-0.0002 -0.0036 0.995 0.0999>}
 dyn_keychain_f {parent=deform_keychain_f, loc=<3.8249 0 0>, rot=<0 0 0 1>}
 dyn_keychain_1_f {parent=deform_keychain_f, loc=<8.1553 0 0>, rot=<0 0 0 1>}
 deform_keychain_f_l {parent=spine_01, loc=<0.5205 7.3117 -7.9133>, rot=<0.3404 0.1131 0.9208 0.1531>}
 dyn_keychain_f_l {parent=deform_keychain_f_l, loc=<0 0 0>, rot=<0 0 0 1>}
 deform_keychain_f_r {parent=spine_01, loc=<0.379 8.5968 5.1214>, rot=<-0.3734 -0.0328 0.9205 0.11>}
 dyn_keychain_f_r {parent=deform_keychain_f_r, loc=<0 0 0>, rot=<0 0 0 1>}
 thigh_l {parent=pelvis, loc=<-1.0855 -0.2109 -8.7053>, rot=<0.0009 0.0355 0.0155 -0.9993>}
 calf_l {parent=thigh_l, loc=<-42.5284 -0.0019 -0.7297>, rot=<0.0006 -0.0161 -0.0391 -0.9991>}
 foot_l {parent=calf_l, loc=<-46.7401 -2.5014 -0.592>, rot=<0.0427 -0.0238 0.0353 -0.9982>}
 ball_l {parent=foot_l, loc=<-3.5658 10.8066 -0.1544>, rot=<0.0001 -0.0001 -0.7063 -0.7079>}
 deform_foot_l {parent=foot_l, loc=<0 0 0>, rot=<0 0 0 1>}
 dyn_foot_l {parent=deform_foot_l, loc=<1.478 -1.1602 -0.4934>, rot=<0.7296 -0.6811 0.0026 -0.0617>}
 calf_twist_02_l {parent=calf_l, loc=<-19.1546 -1.7943 -0.2638>, rot=<0.0237 0.0095 -0.0035 -0.9997>}
 calf_twist_01_l {parent=calf_l, loc=<-35.8483 -1.9197 -0.5794>, rot=<0.0237 0.0095 -0.0035 -0.9997>}
 deform_knee_in_base_l {parent=calf_l, loc=<0.8031 -5.0231 0.2456>, rot=<0 0 0 -1>}
 deform_knee_in_l {parent=deform_knee_in_base_l, loc=<0 0 0>, rot=<0 0 0 1>}
 deform_knee_out_base_l {parent=calf_l, loc=<0 4 0>, rot=<0 0 0 -1>}
 deform_knee_out_l {parent=deform_knee_out_base_l, loc=<0 0 0>, rot=<0 0 0 1>}
 deform_bow_l {parent=calf_l, loc=<-9.4295 4.3097 -0.3779>, rot=<0 0 1 0>}
 dyn_bow_l {parent=deform_bow_l, loc=<0 0 0>, rot=<0 0 0 1>}
 dyn_bow {parent=deform_bow_l, loc=<0 0 0>, rot=<0 0 0 1>}
 thigh_twist_01_l {parent=thigh_l, loc=<-27.952 0 0>, rot=<0.0013 0 0 -1>}
 deform_groin_base_l {parent=thigh_l, loc=<0.4801 5.6776 -0.2486>, rot=<0 0 0 -1>}
 deform_groin_l {parent=deform_groin_base_l, loc=<0 0 0>, rot=<0 0 0 1>}
 thigh_r {parent=pelvis, loc=<-1.0855 -0.2109 8.7053>, rot=<0.0355 -0.0009 0.9993 0.0155>}
 calf_r {parent=thigh_r, loc=<42.5284 0.0019 0.7297>, rot=<0.0006 -0.0161 -0.0391 -0.9991>}
 foot_r {parent=calf_r, loc=<46.7401 2.5014 0.592>, rot=<0.0427 -0.0238 0.0353 -0.9982>}
 ball_r {parent=foot_r, loc=<3.5658 -10.8066 0.1544>, rot=<0.0001 -0.0001 -0.7063 -0.7079>}
 deform_foot_r {parent=foot_r, loc=<0 0 0>, rot=<0 0 0 1>}
 dyn_foot_r {parent=deform_foot_r, loc=<-1.478 1.1603 0.4934>, rot=<-0.6811 -0.7296 -0.0617 -0.0026>}
 calf_twist_02_r {parent=calf_r, loc=<19.1546 1.7943 0.2638>, rot=<0.0237 0.0095 -0.0035 -0.9997>}
 calf_twist_01_r {parent=calf_r, loc=<35.8483 1.9197 0.5794>, rot=<0.0237 0.0095 -0.0035 -0.9997>}
 deform_knee_in_base_r {parent=calf_r, loc=<-0.8031 5.0231 -0.2457>, rot=<0 0 0 -1>}
 deform_knee_in_r {parent=deform_knee_in_base_r, loc=<0 0 0>, rot=<0 0 0 1>}
 deform_knee_out_base_r {parent=calf_r, loc=<0 -4 0>, rot=<0 0 0 -1>}
 deform_knee_out_r {parent=deform_knee_out_base_r, loc=<0 0 0>, rot=<0 0 0 1>}
 deform_bow_r {parent=calf_r, loc=<9.4294 -4.3097 0.3779>, rot=<0 0 0 -1>}
 dyn_bow_r {parent=deform_bow_r, loc=<0 0 0>, rot=<0 0 0 -1>}
 dyn_bow1 {parent=deform_bow_r, loc=<0 0 0>, rot=<0 0 0 -1>}
 thigh_twist_01_r {parent=thigh_r, loc=<27.952 0 0>, rot=<0.0013 0 0 -1>}
 deform_groin_base_r {parent=thigh_r, loc=<-0.4801 -5.6776 0.2486>, rot=<0 0 0 -1>}
 deform_groin_r {parent=deform_groin_base_r, loc=<0 0 0>, rot=<0 0 0 1>}
 deform_glute_base_l {parent=pelvis, loc=<-7.3466 -9.0213 -9.1245>, rot=<0.0172 0.031 -0.4477 -0.8935>}
 deform_glute_l {parent=deform_glute_base_l, loc=<0 0 0>, rot=<0 0 0 1>}
 deform_glute_base_r {parent=pelvis, loc=<-7.3464 -9.0213 9.1245>, rot=<0.031 -0.0172 0.8935 -0.4477>}
 deform_glute_r {parent=deform_glute_base_r, loc=<0 0 0>, rot=<0 0 0 1>}
]
