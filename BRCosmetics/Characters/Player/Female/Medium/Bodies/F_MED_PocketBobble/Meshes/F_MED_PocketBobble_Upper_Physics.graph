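# Physics asset for the upper mesh of the PocketBobble female cosmetic body, covering its skirt, back bow, keychain and dangling toys.
# Sphere and capsule bodies form deform-to-dyn bone chains rooted on spine_01, though two left-front skirt chains skip the deform link.
# Skirt joints have tight angular limits, while accessory deform bodies rotate freely against the spine.
# Cross-constraints tie neighboring front skirt tips together and link the bow to the keychain with twist locked.
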PhysicsAsset /BRCosmetics/Characters/Player/Female/Medium/Bodies/F_MED_PocketBobble/Meshes/F_MED_PocketBobble_Upper_Physics
bodies {
 deform_bow_bk [{shapeType=Sphere, radius=3, shapeName=deform_bow_bk_sphere}]
 deform_keychain_bk [{shapeType=Sphere, radius=1, shapeName=deform_keychain_bk_sphere}]
 deform_skirt_b_l [{shapeType=Sphere, radius=5, shapeName=deform_skirt_b_l_sphere, center=<6.9496 3.333 0>}]
 deform_skirt_b_r [{shapeType=Sphere, radius=5, shapeName=deform_skirt_b_r_sphere, center=<6.9496 3.333 0>}]
 deform_skirt_bk_l [{shapeType=Sphere, radius=5, shapeName=deform_skirt_bk_l_sphere, center=<6.9496 3.333 0>}]
 deform_skirt_bk_r [{shapeType=Sphere, radius=5, shapeName=deform_skirt_bk_r_sphere, center=<6.9496 3.333 0>}]
 deform_skirt_f_m_r [{shapeType=Capsule, length=4, radius=6, shapeName=deform_skirt_f_m_r_capsule, center=<6.414 4.5955 0>}]
 deform_skirt_f_r [{shapeType=Sphere, radius=5, shapeName=deform_skirt_f_r_sphere, center=<6.414 4.84 0>}]
 deform_skirt_m_r [{shapeType=Sphere, radius=3, shapeName=deform_skirt_m_r_sphere, center=<6.3606 3.079 0>}]
 deform_toy_1_r [{shapeType=Capsule, length=0.505, radius=0.505, shapeName=deform_toy_1_r_capsule, rotation=<rot 0 -90 0>}]
 deform_toy_2_r [{shapeType=Capsule, length=0.505, radius=0.505, shapeName=deform_toy_2_r_capsule, rotation=<rot 0 -90 0>}]
 deform_toy_bottom_l [{shapeType=Sphere, radius=0.505, shapeName=deform_toy_bottom_l_sphere}]
 dyn_bow_bk [{shapeType=Capsule, length=6.5, radius=3, shapeName=dyn_bow_bk_capsule, center=<3.3127 -1.9839 0>, rotation=<rot 90 0 0>}]
 dyn_bow_bk_3 [{shapeType=Capsule, length=7.5, radius=3, shapeName=dyn_bow_bk_3_capsule, center=<4.2493 -1.9801 0>, rotation=<rot 90 0 0>}]
 dyn_bow_bk_4 [{shapeType=Capsule, length=7.5, radius=3, shapeName=dyn_bow_bk_4_capsule, center=<4.2493 -1.9801 0>, rotation=<rot 90 0 0>}]
 dyn_keychain_bk [
  {shapeType=Capsule, length=8.75, radius=3, shapeName=dyn_keychain_bk_capsule, center=<4.0419 -2.5469 0>}
  {shapeType=Capsule, length=9.75, radius=3, shapeName=dyn_keychain_bk_capsule, center=<9.5045 -2.9045 0.0267>, rotation=<rot 0 0 -5.1207>}
 ]
 dyn_skirt_b_1_l [{shapeType=Capsule, length=12.25, radius=6, shapeName=dyn_skirt_b_1_l_capsule, center=<4.9 2.5461 0>}]
 dyn_skirt_b_1_r [{shapeType=Capsule, length=6, radius=6, shapeName=dyn_skirt_b_1_r_capsule, center=<4.9 2.5461 0>}]
 dyn_skirt_b_l [{shapeType=Capsule, length=4.75, radius=6, shapeName=dyn_skirt_b_l_capsule, center=<6.414 3.7384 0>}]
 dyn_skirt_b_r [{shapeType=Capsule, length=2, radius=6, shapeName=dyn_skirt_b_r_capsule, center=<6.414 3.7384 0>}]
 dyn_skirt_bk_1_l [{shapeType=Capsule, length=13.75, radius=6, shapeName=dyn_skirt_bk_1_l_capsule, center=<4.9 2.5461 0>}]
 dyn_skirt_bk_1_r [{shapeType=Capsule, length=13.75, radius=6, shapeName=dyn_skirt_bk_1_r_capsule, center=<4.9 2.5461 0>}]
 dyn_skirt_bk_l [{shapeType=Capsule, length=8.25, radius=6, shapeName=dyn_skirt_bk_l_capsule, center=<6.414 3.7384 -1.5979>}]
 dyn_skirt_bk_r [{shapeType=Capsule, length=8.25, radius=6, shapeName=dyn_skirt_bk_r_capsule, center=<6.414 3.7384 1.9522>}]
 dyn_skirt_f_1_l [{shapeType=Capsule, length=6, radius=6, shapeName=dyn_skirt_f_1_l_capsule, center=<4.9 4.9653 0>}]
 dyn_skirt_f_1_r [{shapeType=Capsule, length=6, radius=6, shapeName=dyn_skirt_f_1_r_capsule, center=<4.9 2.5461 0>}]
 dyn_skirt_f_l [{shapeType=Capsule, length=3.75, radius=6, shapeName=dyn_skirt_f_l_capsule, center=<6.414 11.251 0>}]
 dyn_skirt_f_m_1_l [{shapeType=Capsule, length=6, radius=6, shapeName=dyn_skirt_f_m_1_l_capsule, center=<4.9 6.0399 0>}]
 dyn_skirt_f_m_1_r [{shapeType=Capsule, length=6, radius=6, shapeName=dyn_skirt_f_m_1_r_capsule, center=<4.9 3.17 0>}]
 dyn_skirt_f_m_l [{shapeType=Capsule, length=4, radius=6, shapeName=dyn_skirt_f_m_l_capsule, center=<6.3453 8.5809 0>, rotation=<rot 0.0012 0.988 -0.0006>}]
 dyn_skirt_f_m_r [{shapeType=Capsule, length=3.75, radius=5.5, shapeName=dyn_skirt_f_m_r_capsule, center=<6.414 4.5955 0>}]
 dyn_skirt_f_r [{shapeType=Capsule, length=4.25, radius=5.5, shapeName=dyn_skirt_f_r_capsule, center=<6.414 5.401 0>}]
 dyn_skirt_m_r [{shapeType=Capsule, length=2.75, radius=6, shapeName=dyn_skirt_m_r_capsule, center=<6.414 3.3922 0>}]
 dyn_skirt_toy_bottom_l [{shapeType=Capsule, length=2.21, radius=2.9625, shapeName=dyn_skirt_toy_bottom_l_capsule, center=<5.8837 -2.7528 0>, rotation=<rot 90 0 0>}]
 dyn_toy_1_r [{shapeType=Capsule, length=4.295, radius=2.5841, shapeName=dyn_toy_1_r_capsule, center=<4.8338 -2.9586 0>, rotation=<rot 90 0 0>}]
 dyn_toy_2_r [{
   shapeType=Capsule
   length=9.507
   radius=4.1862
   shapeName=dyn_toy_2_r_capsule
   center=<2.436 -4.5308 -0.0084>
   rotation=<rot 0 92.1596 -89.4464>
  }]
 pelvis [{shapeType=Capsule, length=11.932, radius=8.4598}]
 spine_01 [{shapeType=Capsule, length=8.755, radius=8.255, shapeName=spine_01_capsule, rotation=<rot 0 -90 0>}]
}
constraints [
 {bone1Name=deform_skirt_b_l, bone2Name=spine_01, swing1LimitDegrees=35, swing2LimitDegrees=10, twistLimitDegrees=5}
 {bone1Name=dyn_skirt_b_l, bone2Name=deform_skirt_b_l, swing1LimitDegrees=35, swing2LimitDegrees=10, twistLimitDegrees=5}
 {bone1Name=dyn_skirt_b_1_l, bone2Name=dyn_skirt_b_l, swing1LimitDegrees=35, swing2LimitDegrees=10, twistLimitDegrees=5}
 {bone1Name=deform_skirt_b_r, bone2Name=spine_01, swing1LimitDegrees=35, swing2LimitDegrees=10, twistLimitDegrees=5}
 {bone1Name=dyn_skirt_b_r, bone2Name=deform_skirt_b_r, swing1LimitDegrees=35, swing2LimitDegrees=10, twistLimitDegrees=5}
 {bone1Name=dyn_skirt_b_1_r, bone2Name=dyn_skirt_b_r, swing1LimitDegrees=35, swing2LimitDegrees=10, twistLimitDegrees=5}
 {bone1Name=deform_skirt_bk_l, bone2Name=spine_01, swing1LimitDegrees=35, swing2LimitDegrees=10, twistLimitDegrees=5}
 {bone1Name=dyn_skirt_bk_l, bone2Name=deform_skirt_bk_l, swing1LimitDegrees=35, swing2LimitDegrees=10, twistLimitDegrees=5}
 {bone1Name=dyn_skirt_bk_1_l, bone2Name=dyn_skirt_bk_l, swing1LimitDegrees=35, swing2LimitDegrees=10, twistLimitDegrees=5}
 {bone1Name=deform_skirt_bk_r, bone2Name=spine_01, swing1LimitDegrees=35, swing2LimitDegrees=10, twistLimitDegrees=5}
 {bone1Name=dyn_skirt_bk_r, bone2Name=deform_skirt_bk_r, swing1LimitDegrees=35, swing2LimitDegrees=10, twistLimitDegrees=5}
 {bone1Name=dyn_skirt_bk_1_r, bone2Name=dyn_skirt_bk_r, swing1LimitDegrees=35, swing2LimitDegrees=10, twistLimitDegrees=5}
 {bone1Name=dyn_skirt_f_l, bone2Name=spine_01, swing1LimitDegrees=35, swing2LimitDegrees=10, twistLimitDegrees=5}
 {bone1Name=dyn_skirt_f_1_l, bone2Name=dyn_skirt_f_l, swing1LimitDegrees=35, swing2LimitDegrees=10, twistLimitDegrees=5}
 {bone1Name=dyn_skirt_f_m_l, bone2Name=spine_01, swing1LimitDegrees=35, swing2LimitDegrees=10, twistLimitDegrees=5}
 {bone1Name=dyn_skirt_f_m_1_l, bone2Name=dyn_skirt_f_m_l, swing1LimitDegrees=35, swing2LimitDegrees=10, twistLimitDegrees=5}
 {bone1Name=deform_skirt_f_m_r, bone2Name=spine_01, swing1LimitDegrees=35, swing2LimitDegrees=10, twistLimitDegrees=5}
 {bone1Name=dyn_skirt_f_m_r, bone2Name=deform_skirt_f_m_r, swing1LimitDegrees=35, swing2LimitDegrees=10, twistLimitDegrees=5}
 {bone1Name=dyn_skirt_f_m_1_r, bone2Name=dyn_skirt_f_m_r, swing1LimitDegrees=35, swing2LimitDegrees=10, twistLimitDegrees=5}
 {bone1Name=deform_skirt_f_r, bone2Name=spine_01, swing1LimitDegrees=35, swing2LimitDegrees=10, twistLimitDegrees=5}
 {bone1Name=dyn_skirt_f_r, bone2Name=deform_skirt_f_r, swing1LimitDegrees=35, swing2LimitDegrees=10, twistLimitDegrees=5}
 {bone1Name=dyn_skirt_f_1_r, bone2Name=dyn_skirt_f_r, swing1LimitDegrees=35, swing2LimitDegrees=10, twistLimitDegrees=5}
 {bone1Name=deform_skirt_m_r, bone2Name=spine_01, swing1LimitDegrees=35, swing2LimitDegrees=10, twistLimitDegrees=5}
 {bone1Name=dyn_skirt_m_r, bone2Name=deform_skirt_m_r, swing1LimitDegrees=35, swing2LimitDegrees=10, twistLimitDegrees=5}
 {bone1Name=dyn_skirt_f_m_1_l, bone2Name=dyn_skirt_f_1_l, swing1LimitDegrees=15, swing2LimitDegrees=15, twistLimitDegrees=5}
 {bone1Name=dyn_skirt_f_m_1_r, bone2Name=dyn_skirt_f_1_r, swing1LimitDegrees=35, swing2LimitDegrees=35, twistLimitDegrees=35}
 {
  bone1Name=deform_toy_1_r
  bone2Name=spine_01
  swing1LimitDegrees=45
  swing1Motion=Free
  swing2LimitDegrees=45
  swing2Motion=Free
  twistLimitDegrees=45
  twistMotion=Free
 }
 {bone1Name=dyn_toy_1_r, bone2Name=deform_toy_1_r, swing1LimitDegrees=35, swing2LimitDegrees=25, twistLimitDegrees=15}
 {
  bone1Name=deform_toy_2_r
  bone2Name=spine_01
  swing1LimitDegrees=45
  swing1Motion=Free
  swing2LimitDegrees=45
  swing2Motion=Free
  twistLimitDegrees=45
  twistMotion=Free
 }
 {bone1Name=dyn_toy_2_r, bone2Name=deform_toy_2_r, swing1LimitDegrees=35, swing2LimitDegrees=25, twistLimitDegrees=15}
 {bone1Name=dyn_bow_bk, bone2Name=deform_bow_bk, swing1LimitDegrees=35, swing2LimitDegrees=30, twistLimitDegrees=5}
 {bone1Name=dyn_bow_bk_3, bone2Name=dyn_bow_bk, swing1LimitDegrees=35, swing2LimitDegrees=25, twistLimitDegrees=5}
 {bone1Name=dyn_bow_bk_4, bone2Name=dyn_bow_bk, swing1LimitDegrees=35, swing2LimitDegrees=25, twistLimitDegrees=5}
 {
  bone1Name=deform_bow_bk
  bone2Name=spine_01
  swing1LimitDegrees=45
  swing1Motion=Free
  swing2LimitDegrees=45
  swing2Motion=Free
  twistLimitDegrees=45
  twistMotion=Free
 }
 {bone1Name=dyn_skirt_toy_bottom_l, bone2Name=deform_toy_bottom_l, swing1LimitDegrees=45, swing2LimitDegrees=45, twistLimitDegrees=15}
 {bone1Name=dyn_keychain_bk, bone2Name=deform_keychain_bk, swing1LimitDegrees=45, swing2LimitDegrees=15, twistLimitDegrees=1}
 {
  bone1Name=deform_keychain_bk
  bone2Name=spine_01
  swing1LimitDegrees=45
  swing1Motion=Free
  swing2LimitDegrees=45
  swing2Motion=Free
  twistLimitDegrees=45
  twistMotion=Free
 }
 {
  bone1Name=deform_toy_bottom_l
  bone2Name=spine_01
  swing1LimitDegrees=45
  swing1Motion=Free
  swing2LimitDegrees=45
  swing2Motion=Free
  twistLimitDegrees=45
  twistMotion=Free
 }
 {bone1Name=dyn_bow_bk, bone2Name=dyn_keychain_bk, swing1LimitDegrees=45, swing2LimitDegrees=45, twistLimitDegrees=45, twistMotion=Locked}
 {
  bone1Name=spine_01
  bone2Name=pelvis
  swing1LimitDegrees=45
  swing1Motion=Free
  swing2LimitDegrees=45
  swing2Motion=Free
  twistLimitDegrees=45
  twistMotion=Free
 }
]
