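# Animation Blueprint for the F_MED_PocketBobble cosmetic body, adding secondary motion to bows, skirt and keychain.
# It copies the parent mesh pose through a state-aware Control Rig, RigidBody physics, AnimDynamics springs and a post Control Rig, then caches it.
# Two differently tuned upper-body RigidBody passes, selected by bIsInFrontEnd, feed final AnimDynamics springs.
# A gravity-override function updates variables each frame, but the physics nodes show fixed gravity instead of reading them.
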AnimBlueprint /BRCosmetics/Characters/Player/Female/Medium/Bodies/F_MED_PocketBobble/Meshes/F_MED_PocketBobble_AnimBP
target_skeleton=&/F_MED_PocketBobble_Skeleton
graphs [
 AnimGraph {nodes [
   n1 = RigidBody!(
    ComponentPose=Local To Component(LocalPose=Use cached pose 'MainPose_1'())
    Alpha=Get Alpha() -> (Alpha:float)
    OverrideWorldGravity:Vector="0,0,-980"
    SimSpaceSettings:Sim Space Settings Structure=(WorldAlpha=1,VelocityScaleZ=0.3,DampingAlpha=1,MaxLinearVelocity=5000,MaxAngularVelocity=5000,MaxLinearAcceleration=1500,MaxAngularAcceleration=500,ExternalLinearDragV=(X=0.2,Y=0.2,Z=0),ExternalLinearVelocity=(X=0,Y=0,Z=0),ExternalAngularVelocity=(X=0,Y=0,Z=0))
   ) -> (Pose) {
    bClampLinearTranslationLimitToRefPose=true
    bOverrideWorldGravity=true
    overridePhysicsAsset=&/F_MED_PocketBobble_Upper_Physics
    simulationSpace=BaseBoneSpace
    baseBoneRef {boneName=spine_02}
    overrideWorldGravity {z=-980}
    simSpaceSettings {
     maxAngularAcceleration=500
     maxAngularVelocity=5000
     maxLinearAcceleration=1500
     maxLinearVelocity=5000
     velocityScaleZ=0.3
     worldAlpha=1
     externalLinearDragV=<0.2 0.2>
    }
   }
   n2 = Component To Local(ComponentPose=@n1)
   n3 = RigidBody!(ComponentPose=Local To Component(LocalPose=Use cached pose 'MainPose_1'()), Alpha=Get Alpha(), OverrideWorldGravity="0,0,-980") -> (Pose) {
    bClampLinearTranslationLimitToRefPose=true
    bOverrideWorldGravity=true
    overridePhysicsAsset=&F_MED_PocketBobble_Upper_Physics
    simulationSpace=BaseBoneSpace
    baseBoneRef {boneName=spine_02}
    overrideWorldGravity {z=-980}
    simSpaceSettings {
     maxAngularAcceleration=300
     maxAngularVelocity=60
     maxLinearAcceleration=1500
     maxLinearVelocity=500
     velocityScaleZ=0.5
     worldAlpha=0.1
     externalLinearDragV=<5 2 1>
    }
   }
   n4 = Component To Local(ComponentPose=@n3)
   n5 = Get bIsInFrontEnd()
   n6 = Anim Dynamics - Bone: dyn_keychain_f!(
    ComponentPose=Local To Component(LocalPose=Blend Poses by bool(BlendPose_0=@n4, BlendPose_1=@n2, bActiveValue=@n5))
    Alpha=1
    GravityOverride:Vector="0,0,0"
   ) -> (Pose) {
    angularDampingOverride=0.9999
    angularSpringConstant=500
    bAngularSpring=true
    bLinearSpring=true
    bOverrideAngularDamping=true
    bOverrideLinearDamping=true
    bUseGravityOverride=true
    linearDampingOverride=1
    linearSpringConstant=100
    simulationSpace=RootRelative
    boundBone {boneName=dyn_keychain_f}
    physicsBodyDefinitions [{
      boundBone {boneName=dyn_keychain_f}
      boxExtents=<7 2 12>
      constraintSetup {angularLimitsMax=<0 0 0>, angularLimitsMin=<0 0 0>, angularTarget=<1 0 0>, linearAxesMax=<0 0 2>, linearAxesMin=<-3 -5 -2>}
      localJointOffset=<0.02 0 0>
     }]
    relativeSpaceBone {boneName=head}
   }
   n7 = Anim Dynamics - Bone: dyn_keychain_1_f!(ComponentPose=@n6, Alpha=1, GravityOverride:Vector="0,0,0") -> (Pose) {
    angularDampingOverride=0.9999
    angularSpringConstant=700
    bAngularSpring=true
    bLinearSpring=true
    bOverrideAngularDamping=true
    bOverrideLinearDamping=true
    bUseGravityOverride=true
    linearDampingOverride=0.9999
    linearSpringConstant=200
    simulationSpace=RootRelative
    boundBone {boneName=dyn_keychain_1_f}
    physicsBodyDefinitions [{
      boundBone {boneName=dyn_keychain_1_f}
      boxExtents=<10 4 10>
      constraintSetup {angularLimitsMax=<0 0 0>, angularLimitsMin=<0 0 0>, angularTarget=<1 0 0>, linearAxesMax=<0 0 2>, linearAxesMin=<-3 -5 -2>}
      localJointOffset=<0.1 0 0>
     }]
    relativeSpaceBone {boneName=head}
   }
   n8 = Anim Dynamics - Bone: dyn_skirt_toy_l!(ComponentPose=@n7, Alpha=1, GravityOverride:Vector="0,0,0") -> (Pose) {
    angularDampingOverride=0.9999
    angularSpringConstant=500
    bAngularSpring=true
    bOverrideAngularDamping=true
    bUseGravityOverride=true
    simulationSpace=RootRelative
    boundBone {boneName=dyn_skirt_toy_l}
    physicsBodyDefinitions [{
      boundBone {boneName=dyn_skirt_toy_l}
      boxExtents=<10 2 10>
      constraintSetup {angularLimitsMax=<5 5 0>, angularLimitsMin=<-5 -5 0>, angularTarget=<1 0 0>}
      localJointOffset=<0.1 0 0>
     }]
    relativeSpaceBone {boneName=head}
   }
   n9 = Anim Dynamics - Bone: deform_bow_upper!(ComponentPose=@n8, Alpha=1, GravityOverride:Vector="0,0,0") -> (Pose) {
    angularDampingOverride=0.9999
    angularSpringConstant=500
    bAngularSpring=true
    bOverrideAngularDamping=true
    bUseGravityOverride=true
    simulationSpace=RootRelative
    boundBone {boneName=deform_bow_upper}
    physicsBodyDefinitions [{
      boundBone {boneName=deform_bow_upper}
      boxExtents=<10 2 10>
      constraintSetup {angularLimitsMax=<0 1 0>, angularLimitsMin=<0 -1 -20>, angularTarget=<1 0 0>}
      localJointOffset=<0.1 0 0>
     }]
    relativeSpaceBone {boneName=head}
   }
   n10 = Anim Dynamics - Bone: dyn_bow_upper_r!(ComponentPose=@n9, Alpha=1, GravityOverride:Vector="0,0,0") -> (Pose) {
    angularDampingOverride=0.9999
    angularSpringConstant=500
    bAngularSpring=true
    bOverrideAngularDamping=true
    bUseGravityOverride=true
    simulationSpace=RootRelative
    boundBone {boneName=dyn_bow_upper_r}
    physicsBodyDefinitions [{
      boundBone {boneName=dyn_bow_upper_r}
      boxExtents=<10 2 10>
      constraintSetup {angularLimitsMax=<0 5 0>, angularLimitsMin=<0 -5 -20>, angularTarget=<1 0 0>}
      localJointOffset=<0.8 0 0>
     }]
    relativeSpaceBone {boneName=head}
   }
   n11 = Anim Dynamics - Bone: dyn_bow_upper_l!(ComponentPose=@n10, Alpha=1, GravityOverride:Vector="0,0,0") -> (Pose) {
    angularDampingOverride=0.9999
    angularSpringConstant=500
    bAngularSpring=true
    bOverrideAngularDamping=true
    bUseGravityOverride=true
    simulationSpace=RootRelative
    boundBone {boneName=dyn_bow_upper_l}
    physicsBodyDefinitions [{
      boundBone {boneName=dyn_bow_upper_l}
      boxExtents=<10 2 10>
      constraintSetup {angularLimitsMax=<0 5 0>, angularLimitsMin=<0 -5 -20>, angularTarget=<1 0 0>}
      localJointOffset=<0.06 0 0>
     }]
    relativeSpaceBone {boneName=head}
   }
   n12 = Anim Dynamics - Bone: dyn_bow_lower_r!(ComponentPose=@n11, Alpha=1, GravityOverride:Vector="0,0,0") -> (Pose) {
    angularDampingOverride=0.9999
    angularSpringConstant=500
    bAngularSpring=true
    bOverrideAngularDamping=true
    bUseGravityOverride=true
    simulationSpace=RootRelative
    boundBone {boneName=dyn_bow_lower_r}
    physicsBodyDefinitions [{
      boundBone {boneName=dyn_bow_lower_r}
      boxExtents=<10 2 3>
      constraintSetup {angularLimitsMax=<0 5 0>, angularLimitsMin=<0 -5 -20>, angularTarget=<1 0 0>}
      localJointOffset=<0.15 0 0>
     }]
    relativeSpaceBone {boneName=head}
   }
   n13 = Anim Dynamics - Bone: dyn_bow_lower_l!(ComponentPose=@n12, Alpha=1, GravityOverride:Vector="0,0,0") -> (Pose) {
    angularDampingOverride=0.9999
    angularSpringConstant=500
    bAngularSpring=true
    bOverrideAngularDamping=true
    bUseGravityOverride=true
    simulationSpace=RootRelative
    boundBone {boneName=dyn_bow_lower_l}
    physicsBodyDefinitions [{
      boundBone {boneName=dyn_bow_lower_l}
      boxExtents=<10 2 3>
      constraintSetup {angularLimitsMax=<0 5 0>, angularLimitsMin=<0 -5 -20>, angularTarget=<1 0 0>}
      localJointOffset=<0.1 0 0>
     }]
    relativeSpaceBone {boneName=head}
   }
   Output Pose(Result=Component To Local(ComponentPose=@n13)) {layerGroup=None}
   n14 = Control Rig!(
    Source=Copy Pose From Mesh() {bUseAttachedParent=true}
    Alpha
    IsSliding:bool=Get bIsBaseballSliding()
    Emotes:bool=Get bIsPlayingEmote()
    Skydive:bool=Get bIsSkyDiving()
   ) -> (Pose) {controlRigAssetReference {blueprintRigClass=&/F_MED_PocketBobble_CtrlRig.F_MED_PocketBobble_CtrlRig_C}}
   n15 = RigidBody!(
    ComponentPose=Local To Component(LocalPose=@n14)
    Alpha=Get Alpha()
    OverrideWorldGravity="0,0,-980"
    SimSpaceSettings:Sim Space Settings Structure=(WorldAlpha=1,VelocityScaleZ=0.3,DampingAlpha=1,MaxLinearVelocity=5000,MaxAngularVelocity=5000,MaxLinearAcceleration=1500,MaxAngularAcceleration=500,ExternalLinearDragV=(X=0.2,Y=0.2,Z=0),ExternalLinearVelocity=(X=0,Y=0,Z=0),ExternalAngularVelocity=(X=0,Y=0,Z=0))
   ) -> (Pose) {
    bClampLinearTranslationLimitToRefPose=true
    bOverrideWorldGravity=true
    overridePhysicsAsset=&/F_MED_PocketBobble_Physics
    simulationSpace=BaseBoneSpace
    baseBoneRef {boneName=spine_02}
    overrideWorldGravity {z=-980}
    simSpaceSettings {
     maxAngularAcceleration=500
     maxAngularVelocity=5000
     maxLinearAcceleration=1500
     maxLinearVelocity=5000
     velocityScaleZ=0.3
     worldAlpha=1
     externalLinearDragV=<0.2 0.2>
    }
   }
   n16 = Anim Dynamics - Bone: dyn_bow_l!(ComponentPose=@n15, Alpha=1, GravityOverride:Vector="0,0,0") -> (Pose) {
    angularDampingOverride=0.9999
    angularSpringConstant=1500
    bAngularSpring=true
    bOverrideAngularDamping=true
    bUseGravityOverride=true
    simulationSpace=RootRelative
    boundBone {boneName=dyn_bow_l}
    physicsBodyDefinitions [{
      boundBone {boneName=dyn_bow_l}
      boxExtents=<10 5 5>
      constraintSetup {angularLimitsMax=<15 15 0>, angularLimitsMin=<-15 -15 -25>, angularTarget=<0 0 0>}
      localJointOffset=<0.3 0 0>
     }]
    relativeSpaceBone {boneName=head}
    simSpaceSettings {maxAngularAcceleration=5000, maxAngularVelocity=400, simSpaceAngularAlpha=0.4}
   }
   n17 = Anim Dynamics - Bone: dyn_bow_r!(ComponentPose=@n16, Alpha=1, GravityOverride:Vector="0,0,0") -> (Pose) {
    angularDampingOverride=0.999
    angularSpringConstant=1000
    bAngularSpring=true
    bOverrideAngularDamping=true
    bUseGravityOverride=true
    simulationSpace=RootRelative
    boundBone {boneName=dyn_bow_r}
    physicsBodyDefinitions [{
      boundBone {boneName=dyn_bow_r}
      boxExtents=<10 5 5>
      constraintSetup {angularLimitsMax=<15 15 0>, angularLimitsMin=<-15 -15 -25>, angularTarget=<0 0 0>}
      localJointOffset=<0.3 0 0>
     }]
    relativeSpaceBone {boneName=head}
    simSpaceSettings {maxAngularAcceleration=5000, maxAngularVelocity=400, simSpaceAngularAlpha=0.4}
   }
   n18 = Anim Dynamics - Bone: dyn_skirt_f_r!(ComponentPose=@n17, Alpha=1, GravityOverride:Vector="0,0,0") -> (Pose) {
    angularDampingOverride=0.9999
    angularSpringConstant=500
    bAngularSpring=true
    bOverrideAngularDamping=true
    bUseGravityOverride=true
    simulationSpace=RootRelative
    boundBone {boneName=dyn_skirt_f_r}
    physicsBodyDefinitions [{
      boundBone {boneName=dyn_skirt_f_r}
      boxExtents=<10 2 10>
      constraintSetup {angularLimitsMax=<0 5 0>, angularLimitsMin=<0 -5 -25>, angularTarget=<0 0 0>}
      localJointOffset=<0.5 0 0>
     }]
    relativeSpaceBone {boneName=head}
    simSpaceSettings {maxAngularAcceleration=5000, maxAngularVelocity=400, simSpaceAngularAlpha=0.4}
   }
   n19 = Anim Dynamics - Bone: dyn_skirt_f_1_r!(ComponentPose=@n18, Alpha=1, GravityOverride:Vector="0,0,0") -> (Pose) {
    angularDampingOverride=0.9999
    angularSpringConstant=500
    bAngularSpring=true
    bOverrideAngularDamping=true
    bUseGravityOverride=true
    simulationSpace=RootRelative
    boundBone {boneName=dyn_skirt_f_1_r}
    physicsBodyDefinitions [{
      boundBone {boneName=dyn_skirt_f_1_r}
      boxExtents=<10 2 10>
      constraintSetup {angularLimitsMax=<0 5 0>, angularLimitsMin=<0 -5 -25>, angularTarget=<0 0 0>}
      localJointOffset=<0.5 0 0>
     }]
    relativeSpaceBone {boneName=head}
    simSpaceSettings {maxAngularAcceleration=5000, maxAngularVelocity=400, simSpaceAngularAlpha=0.4}
   }
   n20 = Anim Dynamics - Bone: dyn_skirt_f_m_r!(ComponentPose=@n19, Alpha=1, GravityOverride:Vector="0,0,0") -> (Pose) {
    angularDampingOverride=0.9999
    angularSpringConstant=500
    bAngularSpring=true
    bOverrideAngularDamping=true
    bUseGravityOverride=true
    simulationSpace=RootRelative
    boundBone {boneName=dyn_skirt_f_m_r}
    physicsBodyDefinitions [{
      boundBone {boneName=dyn_skirt_f_m_r}
      boxExtents=<10 2 10>
      constraintSetup {angularLimitsMax=<0 5 0>, angularLimitsMin=<0 -5 -25>, angularTarget=<0 0 0>}
      localJointOffset=<0.5 0 0>
     }]
    relativeSpaceBone {boneName=head}
    simSpaceSettings {maxAngularAcceleration=5000, maxAngularVelocity=400, simSpaceAngularAlpha=0.4}
   }
   n21 = Anim Dynamics - Bone: dyn_skirt_f_m_1_r!(ComponentPose=@n20, Alpha=1, GravityOverride:Vector="0,0,0") -> (Pose) {
    angularDampingOverride=0.9999
    angularSpringConstant=500
    bAngularSpring=true
    bOverrideAngularDamping=true
    bUseGravityOverride=true
    simulationSpace=RootRelative
    boundBone {boneName=dyn_skirt_f_m_1_r}
    physicsBodyDefinitions [{
      boundBone {boneName=dyn_skirt_f_m_1_r}
      boxExtents=<10 2 10>
      constraintSetup {angularLimitsMax=<0 5 0>, angularLimitsMin=<0 -5 -25>, angularTarget=<0 0 0>}
      localJointOffset=<0.5 0 0>
     }]
    relativeSpaceBone {boneName=head}
    simSpaceSettings {maxAngularAcceleration=5000, maxAngularVelocity=400, simSpaceAngularAlpha=0.4}
   }
   n22 = Anim Dynamics - Bone: dyn_skirt_f_l!(ComponentPose=@n21, Alpha=1, GravityOverride:Vector="0,0,0") -> (Pose) {
    angularDampingOverride=0.9999
    angularSpringConstant=500
    bAngularSpring=true
    bOverrideAngularDamping=true
    bUseGravityOverride=true
    simulationSpace=RootRelative
    boundBone {boneName=dyn_skirt_f_l}
    physicsBodyDefinitions [{
      boundBone {boneName=dyn_skirt_f_l}
      boxExtents=<10 2 10>
      constraintSetup {angularLimitsMax=<0 0 0>, angularLimitsMin=<0 0 -25>, angularTarget=<1 0 0>}
      localJointOffset=<0.5 0 0>
     }]
    relativeSpaceBone {boneName=head}
    simSpaceSettings {maxAngularAcceleration=5000, maxAngularVelocity=400, simSpaceAngularAlpha=0.4}
   }
   n23 = Anim Dynamics - Bone: dyn_skirt_f_1_l!(ComponentPose=@n22, Alpha=1, GravityOverride:Vector="0,0,0") -> (Pose) {
    angularDampingOverride=0.9999
    angularSpringConstant=500
    bAngularSpring=true
    bOverrideAngularDamping=true
    bUseGravityOverride=true
    simulationSpace=RootRelative
    boundBone {boneName=dyn_skirt_f_1_l}
    physicsBodyDefinitions [{
      boundBone {boneName=dyn_skirt_f_1_l}
      boxExtents=<10 2 10>
      constraintSetup {angularLimitsMax=<0 0 0>, angularLimitsMin=<0 0 -25>, angularTarget=<1 0 0>}
      localJointOffset=<0.5 0 0>
     }]
    relativeSpaceBone {boneName=head}
    simSpaceSettings {maxAngularAcceleration=5000, maxAngularVelocity=400, simSpaceAngularAlpha=0.4}
   }
   n24 = Anim Dynamics - Bone: dyn_skirt_f_m_l!(ComponentPose=@n23, Alpha=1, GravityOverride:Vector="0,0,0") -> (Pose) {
    angularDampingOverride=0.9999
    angularSpringConstant=500
    bAngularSpring=true
    bOverrideAngularDamping=true
    bUseGravityOverride=true
    simulationSpace=RootRelative
    boundBone {boneName=dyn_skirt_f_m_l}
    physicsBodyDefinitions [{
      boundBone {boneName=dyn_skirt_f_m_l}
      boxExtents=<10 2 10>
      constraintSetup {angularLimitsMax=<0 0 0>, angularLimitsMin=<0 0 -25>, angularTarget=<1 0 0>}
      localJointOffset=<0.5 0 0>
     }]
    relativeSpaceBone {boneName=head}
    simSpaceSettings {maxAngularAcceleration=5000, maxAngularVelocity=400, simSpaceAngularAlpha=0.4}
   }
   n25 = Anim Dynamics - Bone: dyn_skirt_f_m_1_l!(ComponentPose=@n24, Alpha=1, GravityOverride:Vector="0,0,0") -> (Pose) {
    angularDampingOverride=0.9999
    angularSpringConstant=500
    bAngularSpring=true
    bOverrideAngularDamping=true
    bUseGravityOverride=true
    simulationSpace=RootRelative
    boundBone {boneName=dyn_skirt_f_m_1_l}
    physicsBodyDefinitions [{
      boundBone {boneName=dyn_skirt_f_m_1_l}
      boxExtents=<10 2 10>
      constraintSetup {angularLimitsMax=<0 0 0>, angularLimitsMin=<0 0 -25>, angularTarget=<1 0 0>}
      localJointOffset=<0.5 0 0>
     }]
    relativeSpaceBone {boneName=head}
    simSpaceSettings {maxAngularAcceleration=5000, maxAngularVelocity=400, simSpaceAngularAlpha=0.4}
   }
   n26 = Anim Dynamics - Bone: dyn_skirt_f!(ComponentPose=@n25, Alpha=1, GravityOverride:Vector="0,0,0") -> (Pose) {
    angularDampingOverride=0.9999
    angularSpringConstant=500
    bAngularSpring=true
    bOverrideAngularDamping=true
    bUseGravityOverride=true
    simulationSpace=RootRelative
    boundBone {boneName=dyn_skirt_f}
    physicsBodyDefinitions [{
      boundBone {boneName=dyn_skirt_f}
      boxExtents=<10 2 10>
      constraintSetup {angularLimitsMax=<0 5 0>, angularLimitsMin=<0 -5 -25>, angularTarget=<0 0 0>}
      localJointOffset=<0.5 0 0>
     }]
    relativeSpaceBone {boneName=head}
    simSpaceSettings {maxAngularAcceleration=5000, maxAngularVelocity=400, simSpaceAngularAlpha=0.4}
   }
   n27 = Anim Dynamics - Bone: dyn_skirt_f_1!(ComponentPose=@n26, Alpha=1, GravityOverride:Vector="0,0,0") -> (Pose) {
    angularDampingOverride=0.9999
    angularSpringConstant=500
    bAngularSpring=true
    bOverrideAngularDamping=true
    bUseGravityOverride=true
    simulationSpace=RootRelative
    boundBone {boneName=dyn_skirt_f_1}
    physicsBodyDefinitions [{
      boundBone {boneName=dyn_skirt_f_1}
      boxExtents=<10 2 10>
      constraintSetup {angularLimitsMax=<0 5 0>, angularLimitsMin=<0 -5 -25>, angularTarget=<0 0 0>}
      localJointOffset=<0.5 0 0>
     }]
    relativeSpaceBone {boneName=head}
    simSpaceSettings {maxAngularAcceleration=5000, maxAngularVelocity=400, simSpaceAngularAlpha=0.4}
   }
   n28 = Control Rig!(Source=Component To Local(ComponentPose=@n27), Alpha) -> (Pose) {controlRigAssetReference {blueprintRigClass=&/F_MED_PocketBobble_Post_CtrlRig.F_MED_PocketBobble_Post_CtrlRig_C}}
   Save cached pose 'MainPose_1'(Pose=@n28) {cacheName=MainPose_1}
  ]}
 EventGraph {nodes [
   Event BlueprintUpdateAnimation()
   > n1 = "Gravity Override State Function (Parameter Asset) (Nativized)"(
    CharacterPartAnimInstance=Self-Reference()
    GameTimeInSeconds=GetGameTimeInSeconds()
    Parameters=&/F_MED_PocketBobble_Parameters
   )
   > Set Gravity Override(Gravity Override=@n1)
   > Set RBAN Settings Out(RBAN Settings Out=@n1.OutRigidBodyAnimNodeParameters)
   Event BlueprintBeginPlay() {is=title K2Node_Event}
   > Set Alpha(Alpha:float=1) -> (Output_Get:float)
   PrintString!(execute, InString="To String (Name)"(InName=@n1.OutCurrentState), bPrintToScreen, bPrintToLog, TextColor, Duration, Key) -> (then)
  ]}
]
variables [
 Alpha {type=number}
 Gravity Override {type=Vector}
 RBAN Settings Out {type=FortRigidBodyAnimNodeParameters}
 Sim Space Settings {
  type=SimSpaceSettings
  default {
   dampingAlpha=1
   maxAngularAcceleration=10000
   maxAngularVelocity=10000
   maxLinearAcceleration=10000
   maxLinearVelocity=10000
   velocityScaleZ=1
   worldAlpha=0
   externalLinearDragV=<0 0 0>
  }
 }
]
