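# Physics asset for the female PocketBobble FaceAcc_Cape mesh part, covering dynamic hair bone chains alongside upper-body bodies.
# Capsules approximate spine, neck, head, clavicles and upper arms, joined by free-rotation constraints.
# Long left and right hair chains and three front chains hang from freely rotating deform roots on the head, with angle-limited joints along each.
# Extra free constraints link mid-chain segments across sides and to the clavicles.
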PhysicsAsset /BRCosmetics/Characters/Player/Female/Medium/Bodies/F_MED_PocketBobble/Meshes/Parts/F_MED_PocketBobble_FaceAcc_Cape_Physics
bodies {
 clavicle_l [{
   shapeType=Capsule
   length=7
   radius=5.25
   shapeName=clavicle_l_capsule
   center=<5.4672 -1.0855 0.3366>
   rotation=<rot 78.9134 -0.0527 -0.0544>
  }]
 clavicle_r [{
   shapeType=Capsule
   length=7.75
   radius=5.5
   shapeName=clavicle_r_capsule
   center=<-5.435 1.0183 -0.363>
   rotation=<rot 84.6697 -0.1121 179.8867>
  }]
 deform_hair_f [{shapeType=Sphere, radius=4, shapeName=deform_hair_f_sphere, center=<2.3057 -1.6473 0>}]
 deform_hair_f_l [{shapeType=Capsule, length=0.25, radius=3, shapeName=deform_hair_f_l_capsule, rotation=<rot 90 0 0>}]
 deform_hair_f_r [{shapeType=Capsule, length=0.25, radius=3, shapeName=deform_hair_f_r_capsule, rotation=<rot 90 0 0>}]
 deform_hair_l [{shapeType=Capsule, length=0.25, radius=3, shapeName=deform_hair_l_capsule, rotation=<rot 90 0 0>}]
 deform_hair_r [{shapeType=Capsule, length=0.25, radius=3, shapeName=deform_hair_r_capsule, rotation=<rot 90 0 0>}]
 dyn_hair_1_l [{shapeType=Capsule, length=5.75, radius=3, shapeName=dyn_hair_1_l_capsule, center=<1.7632 0 0>, rotation=<rot 90 0 0>}]
 dyn_hair_1_r [{shapeType=Capsule, length=5.75, radius=3, shapeName=dyn_hair_1_r_capsule, center=<1.7632 0 0>, rotation=<rot 90 0 0>}]
 dyn_hair_2_l [{shapeType=Capsule, length=8.25, radius=4.25, shapeName=dyn_hair_2_l_capsule, center=<4.9192 0 0>, rotation=<rot 90 0 0>}]
 dyn_hair_2_r [{shapeType=Capsule, length=8.25, radius=4.25, shapeName=dyn_hair_2_r_capsule, center=<4.9192 0 0>, rotation=<rot 90 0 0>}]
 dyn_hair_3_l [{shapeType=Capsule, length=13.5, radius=4.25, shapeName=dyn_hair_3_l_capsule, center=<7.4319 0 0>, rotation=<rot 90 0 0>}]
 dyn_hair_3_r [{shapeType=Capsule, length=13.5, radius=4.25, shapeName=dyn_hair_3_r_capsule, center=<7.4319 0 0>, rotation=<rot 90 0 0>}]
 dyn_hair_4_l [{shapeType=Capsule, length=13.5, radius=4.5, shapeName=dyn_hair_4_l_capsule, center=<7.7203 0 0>, rotation=<rot 90 0 0>}]
 dyn_hair_4_r [{shapeType=Capsule, length=13, radius=5, shapeName=dyn_hair_4_r_capsule, center=<7.7203 0 0>, rotation=<rot 90 0 0>}]
 dyn_hair_5_l [{shapeType=Capsule, length=14.25, radius=3.75, shapeName=dyn_hair_5_l_capsule, center=<7.7203 0 0>, rotation=<rot 90 0 0>}]
 dyn_hair_5_r [{shapeType=Capsule, length=14.25, radius=3.75, shapeName=dyn_hair_5_r_capsule, center=<7.7203 0 0>, rotation=<rot 90 0 0>}]
 dyn_hair_f_1 [{shapeType=Sphere, radius=4, shapeName=dyn_hair_f_1_sphere, center=<2.1916 -1.576 0>}]
 dyn_hair_f_1_l [{shapeType=Capsule, length=9.25, radius=2.5, shapeName=dyn_hair_f_1_l_capsule, center=<4.0775 -2.1856 0>, rotation=<rot 90 0 0>}]
 dyn_hair_f_1_r [{shapeType=Capsule, length=9.25, radius=2.5, shapeName=dyn_hair_f_1_r_capsule, center=<4.0775 -2.1856 0>, rotation=<rot 90 0 0>}]
 dyn_hair_f_2 [{shapeType=Sphere, radius=4, shapeName=dyn_hair_f_2_sphere, center=<2.3057 -1.6473 0>}]
 dyn_hair_f_2_l [{shapeType=Capsule, length=10.25, radius=2.25, shapeName=dyn_hair_f_2_l_capsule, center=<5.2503 -1.0553 0>, rotation=<rot 90 0 0>}]
 dyn_hair_f_2_r [{shapeType=Capsule, length=10.25, radius=2.25, shapeName=dyn_hair_f_2_r_capsule, center=<5.2503 -1.0553 0>, rotation=<rot 90 0 0>}]
 dyn_hair_f_3_l [{
   shapeType=Capsule
   length=7.5
   radius=1.5
   shapeName=dyn_hair_f_3_l_capsule
   center=<3.8748 -0.7673 0.3609>
   rotation=<rot 84.3746 179.9998 5.6248>
  }]
 dyn_hair_f_3_r [{shapeType=Capsule, length=7.5, radius=1.5, shapeName=dyn_hair_f_3_r_capsule, center=<3.8748 -0.7673 0>, rotation=<rot 90 0 0>}]
 head [{shapeType=Capsule, length=4, radius=7.25, shapeName=head_capsule, center=<4.5655 0 0>, rotation=<rot 90 -55.2374 0>}]
 neck_01 [{shapeType=Capsule, length=9.1895, radius=4, shapeName=neck_01_capsule, center=<4.5056 0 0>, rotation=<rot 90 0 0>}]
 spine_01 [{shapeType=Capsule, length=8, radius=11, shapeName=spine_01_capsule, center=<0 0.9056 0>, rotation=<rot 0 -90 0>}]
 spine_02 [{shapeType=Capsule, length=6.505, radius=11.5, shapeName=spine_02_capsule, center=<3.9326 4.7721 0>, rotation=<rot 90 0 0>}]
 spine_03 [{shapeType=Capsule, length=2.5, radius=10.5, shapeName=spine_03_capsule, center=<4.7296 3.9145 0>}]
 spine_04 [{shapeType=Capsule, length=14.5136, radius=15.4864, shapeName=spine_04_capsule, center=<0.3099 -3.1335 0>}]
 spine_05 [{shapeType=Capsule, length=13.5396, radius=16.4604, shapeName=spine_05_capsule, center=<-2.5225 -8.3282 0>}]
 upperarm_l [{shapeType=Capsule, length=18, radius=6.5, shapeName=upperarm_l_capsule, center=<9.9048 0 -0.3109>, rotation=<rot 90 0 0>}]
 upperarm_r [{shapeType=Capsule, length=18, radius=6.5, shapeName=upperarm_r_capsule, center=<-9.9048 0 0.3109>, rotation=<rot 90 180 0>}]
}
constraints [
 {
  bone1Name=spine_04
  bone2Name=spine_03
  swing1LimitDegrees=45
  swing1Motion=Free
  swing2LimitDegrees=45
  swing2Motion=Free
  twistLimitDegrees=45
  twistMotion=Free
 }
 {
  bone1Name=spine_05
  bone2Name=spine_04
  swing1LimitDegrees=45
  swing1Motion=Free
  swing2LimitDegrees=45
  swing2Motion=Free
  twistLimitDegrees=45
  twistMotion=Free
 }
 {
  bone1Name=neck_01
  bone2Name=spine_05
  swing1LimitDegrees=45
  swing1Motion=Free
  swing2LimitDegrees=45
  swing2Motion=Free
  twistLimitDegrees=45
  twistMotion=Free
 }
 {
  bone1Name=head
  bone2Name=neck_01
  swing1LimitDegrees=45
  swing1Motion=Free
  swing2LimitDegrees=45
  swing2Motion=Free
  twistLimitDegrees=45
  twistMotion=Free
 }
 {bone1Name=dyn_hair_f_1_l, bone2Name=deform_hair_f_l, swing1LimitDegrees=10, swing2LimitDegrees=10, twistLimitDegrees=1}
 {
  bone1Name=deform_hair_f_l
  bone2Name=head
  swing1LimitDegrees=45
  swing1Motion=Free
  swing2LimitDegrees=45
  swing2Motion=Free
  twistLimitDegrees=45
  twistMotion=Free
 }
 {bone1Name=dyn_hair_f_2_l, bone2Name=dyn_hair_f_1_l, swing1LimitDegrees=35, swing2LimitDegrees=35, twistLimitDegrees=5}
 {bone1Name=dyn_hair_f_3_l, bone2Name=dyn_hair_f_2_l, swing1LimitDegrees=35, swing2LimitDegrees=35, twistLimitDegrees=5}
 {
  bone1Name=deform_hair_l
  bone2Name=head
  swing1LimitDegrees=45
  swing1Motion=Free
  swing2LimitDegrees=45
  swing2Motion=Free
  twistLimitDegrees=45
  twistMotion=Free
 }
 {bone1Name=dyn_hair_1_l, bone2Name=deform_hair_l, swing1LimitDegrees=15, swing2LimitDegrees=15, twistLimitDegrees=5}
 {bone1Name=dyn_hair_2_l, bone2Name=dyn_hair_1_l, swing1LimitDegrees=35, swing2LimitDegrees=35, twistLimitDegrees=5}
 {bone1Name=dyn_hair_3_l, bone2Name=dyn_hair_2_l, swing1LimitDegrees=35, swing2LimitDegrees=30, twistLimitDegrees=5}
 {bone1Name=dyn_hair_4_l, bone2Name=dyn_hair_3_l, swing1LimitDegrees=35, swing2LimitDegrees=15, twistLimitDegrees=5}
 {bone1Name=dyn_hair_5_l, bone2Name=dyn_hair_4_l, swing1LimitDegrees=35, swing2LimitDegrees=15, twistLimitDegrees=5}
 {
  bone1Name=deform_hair_r
  bone2Name=head
  swing1LimitDegrees=45
  swing1Motion=Free
  swing2LimitDegrees=45
  swing2Motion=Free
  twistLimitDegrees=45
  twistMotion=Free
 }
 {bone1Name=dyn_hair_1_r, bone2Name=deform_hair_r, swing1LimitDegrees=15, swing2LimitDegrees=15, twistLimitDegrees=5}
 {bone1Name=dyn_hair_2_r, bone2Name=dyn_hair_1_r, swing1LimitDegrees=35, swing2LimitDegrees=35, twistLimitDegrees=5}
 {bone1Name=dyn_hair_3_r, bone2Name=dyn_hair_2_r, swing1LimitDegrees=35, swing2LimitDegrees=30, twistLimitDegrees=5}
 {bone1Name=dyn_hair_4_r, bone2Name=dyn_hair_3_r, swing1LimitDegrees=35, swing2LimitDegrees=15, twistLimitDegrees=5}
 {bone1Name=dyn_hair_5_r, bone2Name=dyn_hair_4_r, swing1LimitDegrees=35, swing2LimitDegrees=15, twistLimitDegrees=5}
 {
  bone1Name=deform_hair_f_r
  bone2Name=head
  swing1LimitDegrees=45
  swing1Motion=Free
  swing2LimitDegrees=45
  swing2Motion=Free
  twistLimitDegrees=45
  twistMotion=Free
 }
 {bone1Name=dyn_hair_f_1_r, bone2Name=deform_hair_f_r, swing1LimitDegrees=10, swing2LimitDegrees=10, twistLimitDegrees=5}
 {bone1Name=dyn_hair_f_2_r, bone2Name=dyn_hair_f_1_r, swing1LimitDegrees=35, swing2LimitDegrees=35, twistLimitDegrees=5}
 {bone1Name=dyn_hair_f_3_r, bone2Name=dyn_hair_f_2_r, swing1LimitDegrees=35, swing2LimitDegrees=35, twistLimitDegrees=5}
 {
  bone1Name=clavicle_r
  bone2Name=spine_05
  swing1LimitDegrees=45
  swing1Motion=Free
  swing2LimitDegrees=45
  swing2Motion=Free
  twistLimitDegrees=45
  twistMotion=Free
 }
 {
  bone1Name=upperarm_r
  bone2Name=clavicle_r
  swing1LimitDegrees=45
  swing1Motion=Free
  swing2LimitDegrees=45
  swing2Motion=Free
  twistLimitDegrees=45
  twistMotion=Free
 }
 {
  bone1Name=clavicle_l
  bone2Name=spine_05
  swing1LimitDegrees=45
  swing1Motion=Free
  swing2LimitDegrees=45
  swing2Motion=Free
  twistLimitDegrees=45
  twistMotion=Free
 }
 {
  bone1Name=upperarm_l
  bone2Name=clavicle_l
  swing1LimitDegrees=45
  swing1Motion=Free
  swing2LimitDegrees=45
  swing2Motion=Free
  twistLimitDegrees=45
  twistMotion=Free
 }
 {
  bone1Name=spine_03
  bone2Name=spine_02
  swing1LimitDegrees=45
  swing1Motion=Free
  swing2LimitDegrees=45
  swing2Motion=Free
  twistLimitDegrees=45
  twistMotion=Free
 }
 {
  bone1Name=deform_hair_f
  bone2Name=head
  swing1LimitDegrees=45
  swing1Motion=Free
  swing2LimitDegrees=45
  swing2Motion=Free
  twistLimitDegrees=45
  twistMotion=Free
 }
 {bone1Name=dyn_hair_f_1, bone2Name=deform_hair_f, swing1LimitDegrees=15, swing2LimitDegrees=5, twistLimitDegrees=1}
 {bone1Name=dyn_hair_f_2, bone2Name=dyn_hair_f_1, swing1LimitDegrees=35, swing2LimitDegrees=15, twistLimitDegrees=5}
 {
  bone1Name=spine_02
  bone2Name=spine_01
  swing1LimitDegrees=45
  swing1Motion=Free
  swing2LimitDegrees=45
  swing2Motion=Free
  twistLimitDegrees=45
  twistMotion=Free
 }
 {
  bone1Name=dyn_hair_4_r
  bone2Name=clavicle_r
  swing1LimitDegrees=45
  swing1Motion=Free
  swing2LimitDegrees=45
  swing2Motion=Free
  twistLimitDegrees=45
  twistMotion=Free
 }
 {
  bone1Name=dyn_hair_4_r
  bone2Name=dyn_hair_4_l
  swing1LimitDegrees=45
  swing1Motion=Free
  swing2LimitDegrees=45
  swing2Motion=Free
  twistLimitDegrees=45
  twistMotion=Free
 }
 {
  bone1Name=dyn_hair_4_l
  bone2Name=clavicle_l
  swing1LimitDegrees=45
  swing1Motion=Free
  swing2LimitDegrees=45
  swing2Motion=Free
  twistLimitDegrees=45
  twistMotion=Free
 }
 {
  bone1Name=dyn_hair_3_r
  bone2Name=dyn_hair_3_l
  swing1LimitDegrees=45
  swing1Motion=Free
  swing2LimitDegrees=45
  swing2Motion=Free
  twistLimitDegrees=45
  twistMotion=Free
 }
]
