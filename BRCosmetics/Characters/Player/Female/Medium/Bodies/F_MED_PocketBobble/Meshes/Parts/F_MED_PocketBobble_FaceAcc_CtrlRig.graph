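# Corrective Control Rig for the PocketBobble female character's face accessory part, adjusting hair deform bones as the head moves.
# A spherical pose reader measures the head bone's orientation relative to spine_05 within an active and falloff region.
# Its output weights a Modify Transforms node that applies preset rotations to bones including deform_hair_l and deform_hair_r, carrying their child hair and bow chains along.
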ControlRigBlueprint /BRCosmetics/Characters/Player/Female/Medium/Bodies/F_MED_PocketBobble/Meshes/Parts/F_MED_PocketBobble_FaceAcc_CtrlRig
graphs [
 RigVMFunctionLibrary {links [], nodes []}
 RigVMModel {nodes [
   RigUnit_BeginExecution = &RigUnit_BeginExecution()
   n1 = &RigUnit_SphericalPoseReader(
    ActiveRegionScaleFactors=(PositiveWidth=1,NegativeWidth=1,PositiveHeight=2,NegativeHeight=2)
    ActiveRegionSize=0.15
    Debug=(bDrawDebug=false,bDraw2D=False,bDrawLocalAxes=False,DebugScale=25,DebugSegments=20,DebugThickness=0.25)
    DriverAxis=(X=1,Y=0,Z=0)
    DriverItem=(Type=Bone,Name="head")
    FalloffRegionScaleFactors=(PositiveWidth=1,NegativeWidth=1,PositiveHeight=1,NegativeHeight=1)
    FalloffSize=0.7
    FlipHeightScaling=False
    FlipWidthScaling=False
    OptionalParentItem=(Type=Bone,Name="spine_05")
    RotationOffset=(X=-90,Y=0,Z=-40)
    ExecutePin=@RigUnit_BeginExecution.ExecutePin
   )
   > &RigUnit_ModifyTransforms(
    ItemToModify="((Item=(Type=Bone,Name=\"deform_hair_l\"),Transform=(Rotation=(X=-0.2468,Y=0.0778,Z=-0.2905,W=0.9212),Translation=(X=0,Y=0,Z=0),Scale3D=(X=1,Y=1,Z=1))),(Item=(Type=Bone,Name=\"deform_hair_r\"),Transform=(Rotation=(X=0.2468,Y=-0.0778,Z=-0.2905,W=0.9212),Transla ...(truncated)"
    WeightMinimum=0
    Weight=@n1.OutputParam
   )
  ]}
]
hierarchy [
 root {loc=<0 0 0>, rot=<0 0 0 1>}
 pelvis {parent=root, loc=<0 -0.5301 98.0571>, rot=<0.0127 0.707 -0.0127 -0.707>}
 spine_01 {parent=pelvis, loc=<6.1208 -0.9328 0>, rot=<0 0 0.034 -0.9994>}
 spine_02 {parent=spine_01, loc=<6.5997 -0.2095 0>, rot=<0 0 0.0238 -0.9997>}
 spine_03 {parent=spine_02, loc=<6.4914 -0.1237 0>, rot=<0 0 0.0149 -0.9999>}
 spine_04 {parent=spine_03, loc=<6.9598 0.0794 0>, rot=<0 0 -0.0132 -0.9999>}
 spine_05 {parent=spine_04, loc=<6.9987 0 0>, rot=<0 0 -0.0015 -1>}
 neck_01 {parent=spine_05, loc=<11.3138 0.1529 0>, rot=<0 0 -0.2003 -0.9797>}
 neck_02 {parent=neck_01, loc=<4.5627 0 0>, rot=<0 0 0.0504 -0.9987>}
 head {parent=neck_02, loc=<5.3914 0 0>, rot=<0 0 0.115 -0.9934>}
 deform_hair_l {parent=head, loc=<11.5178 -5.8572 -6.909>, rot=<0.7834 -0.5097 -0.2176 -0.2813>}
 dyn_hair_1_l {parent=deform_hair_l, loc=<0 0 0>, rot=<0 0 0 1>}
 dyn_hair_2_l {parent=dyn_hair_1_l, loc=<3.4731 0 0>, rot=<0 0 -0.4546 -0.8907>}
 dyn_hair_3_l {parent=dyn_hair_2_l, loc=<9.2531 0 0>, rot=<-0.0072 0.0218 -0.4527 -0.8914>}
 dyn_hair_4_l {parent=dyn_hair_3_l, loc=<14.007 0 0>, rot=<0.0035 -0.0387 -0.0687 -0.9969>}
 dyn_hair_5_l {parent=dyn_hair_4_l, loc=<15.9611 0 0>, rot=<-0.0019 0.0638 0.0628 -0.996>}
 dyn_hairbow_l {parent=deform_hair_l, loc=<-0.1955 -0.9678 1.6253>, rot=<0.4411 0.1719 0.0592 -0.8789>}
 dyn_hairbow_1_l {parent=dyn_hairbow_l, loc=<0.1553 -0.263 0.6038>, rot=<-0.1775 0.5945 -0.0978 -0.7782>}
 dyn_hairbow_2_l {parent=dyn_hairbow_l, loc=<0.619 -0.2469 -0.9162>, rot=<0.1096 -0.1881 -0.0497 -0.9747>}
 deform_hair_f_l {parent=head, loc=<12.1625 4.1935 -3.303>, rot=<-0.2526 0.5612 0.7775 0.1296>}
 dyn_hair_f_1_l {parent=deform_hair_f_l, loc=<0 0 0>, rot=<0 0 0 1>}
 dyn_hair_f_2_l {parent=dyn_hair_f_1_l, loc=<8.741 0 0>, rot=<-0.0962 0.0576 -0.3056 -0.9455>}
 dyn_hair_f_3_l {parent=dyn_hair_f_2_l, loc=<10.2326 0 0>, rot=<-0.0655 0.0912 0.0026 -0.9937>}
 deform_hair_f {parent=head, loc=<12.7115 5.978 0.0744>, rot=<0.0069 -0.0006 0.898 0.44>}
 dyn_hair_f_1 {parent=deform_hair_f, loc=<0 0 0>, rot=<0 0 0 1>}
 dyn_hair_f_2 {parent=dyn_hair_f_1, loc=<4.6686 0 0>, rot=<-0.0189 0.0132 -0.3799 -0.9247>}
 deform_bow_f {parent=head, loc=<13.8528 4.3276 0.0789>, rot=<-0.0034 -0.006 0.0288 -0.9996>}
 dyn_bow_f_l {parent=deform_bow_f, loc=<1.67 -0.6034 -2.3406>, rot=<0.0524 -0.6925 0.1003 -0.7124>}
 dyn_bow_f_r {parent=deform_bow_f, loc=<1.6251 -0.5733 2.3884>, rot=<-0.049 0.7019 0.0951 -0.7042>}
 deform_hair_f_r {parent=head, loc=<12.1625 4.1935 3.303>, rot=<0.255 -0.5616 0.7767 0.1278>}
 dyn_hair_f_1_r {parent=deform_hair_f_r, loc=<0 0 0>, rot=<0 0 0 1>}
 dyn_hair_f_2_r {parent=dyn_hair_f_1_r, loc=<8.7408 -0.0001 0>, rot=<0.074 -0.1671 -0.3047 -0.9348>}
 dyn_hair_f_3_r {parent=dyn_hair_f_2_r, loc=<10.2331 0 0>, rot=<0.012 -0.0114 0.0173 -0.9997>}
 deform_hair_r {parent=head, loc=<11.5622 -6.4011 6.2958>, rot=<0.7834 -0.5097 0.2176 0.2813>}
 dyn_hair_1_r {parent=deform_hair_r, loc=<0 0 0>, rot=<0 0 0 1>}
 dyn_hair_2_r {parent=dyn_hair_1_r, loc=<3.4732 -0.0001 0>, rot=<0 0 -0.4546 -0.8907>}
 dyn_hair_3_r {parent=dyn_hair_2_r, loc=<9.2531 0.0001 0>, rot=<0.0072 -0.0218 -0.4527 -0.8914>}
 dyn_hair_4_r {parent=dyn_hair_3_r, loc=<14.0067 0 0>, rot=<-0.0035 0.0387 -0.0687 -0.9969>}
 dyn_hair_5_r {parent=dyn_hair_4_r, loc=<15.9613 0 0>, rot=<0.0019 -0.0638 0.0628 -0.996>}
 dyn_hairbow_r {parent=deform_hair_r, loc=<0.2207 -0.4509 -2.0191>, rot=<-0.465 -0.1771 -0.0297 -0.8669>}
 dyn_hairbow_2_r {parent=dyn_hairbow_r, loc=<0.385 -0.5351 0.8032>, rot=<-0.1573 0.1898 0.056 -0.9675>}
 dyn_hairbow_1_r {parent=dyn_hairbow_r, loc=<-0.0117 -0.7905 -0.8504>, rot=<0.2269 -0.5787 0.0268 -0.7829>}
 clavicle_l {parent=spine_05, loc=<5.4831 -0.1053 -3.2946>, rot=<0.0465 -0.7289 -0.009 -0.6829>}
 upperarm_l {parent=clavicle_l, loc=<9.8658 -1.2136 -0.4105>, rot=<0.0164 -0.3245 -0.0269 -0.9454>}
 upperarm_twist_01_l {parent=upperarm_l, loc=<7.2135 -0.1134 0>, rot=<0 0 0 1>}
 deform_upperarm_bk_base_l {parent=upperarm_twist_01_l, loc=<-6.4053 -4.3746 0.314>, rot=<0 0 0 -1>}
 deform_upperarm_bk_l {parent=deform_upperarm_bk_base_l, loc=<0 0 0>, rot=<0 0 0 1>}
 deform_upperarm_fr_base_l {parent=upperarm_twist_01_l, loc=<-5.6312 5.092 0.3354>, rot=<0 0 0 -1>}
 deform_upperarm_fr_l {parent=deform_upperarm_fr_base_l, loc=<0 0 0>, rot=<0 0 0 1>}
 deform_upperarm_up_base_l {parent=upperarm_twist_01_l, loc=<-8.891 0.4825 3.9517>, rot=<0 0 0 -1>}
 deform_upperarm_up_l {parent=deform_upperarm_up_base_l, loc=<0 0 0>, rot=<0 0 0 1>}
 clavicle_r {parent=spine_05, loc=<5.4831 -0.1053 3.2946>, rot=<0.7289 0.0465 -0.6829 0.009>}
 upperarm_r {parent=clavicle_r, loc=<-9.8658 1.2136 0.4105>, rot=<0.0164 -0.3245 -0.0269 -0.9454>}
 upperarm_twist_01_r {parent=upperarm_r, loc=<-7.2135 0.1134 0>, rot=<0 0 0 -1>}
 deform_upperarm_bk_base_r {parent=upperarm_twist_01_r, loc=<6.4053 4.3746 -0.314>, rot=<0 0 0 -1>}
 deform_upperarm_bk_r {parent=deform_upperarm_bk_base_r, loc=<0 0 0>, rot=<0 0 0 1>}
 deform_upperarm_fr_base_r {parent=upperarm_twist_01_r, loc=<5.6312 -5.092 -0.3354>, rot=<0 0 0 -1>}
 deform_upperarm_fr_r {parent=deform_upperarm_fr_base_r, loc=<0 0 0>, rot=<0 0 0 1>}
 deform_upperarm_up_base_r {parent=upperarm_twist_01_r, loc=<8.891 -0.4825 -3.9517>, rot=<0 0 0 -1>}
 deform_upperarm_up_r {parent=deform_upperarm_up_base_r, loc=<0 0 0>, rot=<0 0 0 1>}
]
