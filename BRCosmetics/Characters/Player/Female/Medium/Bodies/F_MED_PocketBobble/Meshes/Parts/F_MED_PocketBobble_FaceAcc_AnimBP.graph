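# Animation Blueprint for the PocketBobble female character's face accessory mesh, adding secondary motion to hair and bows.
# Blends attached-parent and body-mesh poses at the clavicles, applies hair AnimDynamics and a Control Rig, then caches the result.
# A bool blend selects chained backpack and cape RigidBody simulations or a default one, before spring-driven bow AnimDynamics.
# The event graph updates gravity and rigid-body settings, captures the body mesh and detects back bling and capes.
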AnimBlueprint /BRCosmetics/Characters/Player/Female/Medium/Bodies/F_MED_PocketBobble/Meshes/Parts/F_MED_PocketBobble_FaceAcc_AnimBP
target_skeleton=&/F_MED_PocketBobble_FaceAcc_Skeleton
graphs [
 AnimGraph {nodes [
   n1 = RigidBody!(
    ComponentPose=Local To Component(LocalPose=Use cached pose 'PreRigidPose'())
    Alpha
    OverrideWorldGravity:Vector="0,0,0"
    SimSpaceSettings:Sim Space Settings Structure=(WorldAlpha=0.5,VelocityScaleZ=0.4,DampingAlpha=1,MaxLinearVelocity=5000,MaxAngularVelocity=5000,MaxLinearAcceleration=2500,MaxAngularAcceleration=500,ExternalLinearDragV=(X=1,Y=1,Z=1),ExternalLinearVelocity=(X=0,Y=0,Z=0),ExternalAngularVelocity=(X=0,Y=0,Z=0))
   ) -> (Pose) {
    bClampLinearTranslationLimitToRefPose=true
    bOverrideWorldGravity=true
    overridePhysicsAsset=&/F_MED_PocketBobble_FaceAcc_Backpack_Physics
    simulationSpace=BaseBoneSpace
    alphaScaleBiasClamp {bMapRange=true, outRange {max=0, min=1}}
    baseBoneRef {boneName=head}
    simSpaceSettings {
     maxAngularAcceleration=500
     maxAngularVelocity=5000
     maxLinearAcceleration=2500
     maxLinearVelocity=5000
     velocityScaleZ=0.4
     worldAlpha=0.5
     externalLinearDragV=<1 1 1>
    }
   }
   n2 = RigidBody(
    ComponentPose=@n1
    OverrideWorldGravity="0,0,0"
    SimSpaceSettings=(WorldAlpha=0.5,VelocityScaleZ=0.4,DampingAlpha=1,MaxLinearVelocity=5000,MaxAngularVelocity=5000,MaxLinearAcceleration=2500,MaxAngularAcceleration=500,ExternalLinearDragV=(X=1,Y=1,Z=1),ExternalLinearVelocity=(X=0,Y=0,Z=0),ExternalAngularVelocity=(X=0,Y=0,Z=0))
   ) {
    bClampLinearTranslationLimitToRefPose=true
    bOverrideWorldGravity=true
    overridePhysicsAsset=&/F_MED_PocketBobble_FaceAcc_Cape_Physics
    simulationSpace=BaseBoneSpace
    alphaScaleBiasClamp {outRange {max=0}}
    baseBoneRef {boneName=head}
    simSpaceSettings {
     maxAngularAcceleration=500
     maxAngularVelocity=5000
     maxLinearAcceleration=2500
     maxLinearVelocity=5000
     velocityScaleZ=0.4
     worldAlpha=0.5
     externalLinearDragV=<1 1 1>
    }
   }
   n3 = Component To Local(ComponentPose=@n2)
   n4 = RigidBody(
    ComponentPose=Local To Component(LocalPose=Use cached pose 'PreRigidPose'())
    OverrideWorldGravity="0,0,0"
    SimSpaceSettings=(WorldAlpha=0.5,VelocityScaleZ=0.4,DampingAlpha=1,MaxLinearVelocity=5000,MaxAngularVelocity=5000,MaxLinearAcceleration=2500,MaxAngularAcceleration=500,ExternalLinearDragV=(X=1,Y=1,Z=1),ExternalLinearVelocity=(X=0,Y=0,Z=0),ExternalAngularVelocity=(X=0,Y=0,Z=0))
   ) {
    bClampLinearTranslationLimitToRefPose=true
    bOverrideWorldGravity=true
    overridePhysicsAsset=&/F_MED_PocketBobble_FaceAcc_Physics
    simulationSpace=BaseBoneSpace
    baseBoneRef {boneName=head}
    simSpaceSettings {
     maxAngularAcceleration=500
     maxAngularVelocity=5000
     maxLinearAcceleration=2500
     maxLinearVelocity=5000
     velocityScaleZ=0.4
     worldAlpha=0.5
     externalLinearDragV=<1 1 1>
    }
   }
   n5 = Component To Local(ComponentPose=@n4)
   n6 = Anim Dynamics - Bone: dyn_hairbow_l!(ComponentPose=Local To Component(LocalPose=Blend Poses by bool(BlendPose_0=@n3, BlendPose_1=@n5)), Alpha=1) -> (Pose) {
    angularDampingOverride=0.9999
    angularSpringConstant=5000
    bAngularSpring=true
    bOverrideAngularDamping=true
    simulationSpace=RootRelative
    alphaScaleBiasClamp {scale=0.6}
    boundBone {boneName=dyn_hairbow_l}
    physicsBodyDefinitions [{
      boundBone {boneName=dyn_hairbow_l}
      boxExtents=<10 10 10>
      constraintSetup {angularLimitsMax=<15 15 15>, angularLimitsMin=<-15 -15 -15>, angularTarget=<1 0 0>}
      localJointOffset=<0.01 0 0>
     }]
    relativeSpaceBone {boneName=head}
   }
   n7 = Anim Dynamics - Bone: dyn_hairbow_1_l!(ComponentPose=@n6, Alpha=1) -> (Pose) {
    angularDampingOverride=0.9999
    angularSpringConstant=700
    bAngularSpring=true
    bOverrideAngularDamping=true
    simulationSpace=RootRelative
    alphaScaleBiasClamp {scale=0.6}
    boundBone {boneName=dyn_hairbow_1_l}
    physicsBodyDefinitions [{
      boundBone {boneName=dyn_hairbow_1_l}
      boxExtents=<10 5 5>
      constraintSetup {angularLimitsMax=<15 15 15>, angularLimitsMin=<-15 -15 -15>, angularTarget=<1 0 0>}
      localJointOffset=<0.05 0 0>
     }]
    relativeSpaceBone {boneName=head}
   }
   n8 = Anim Dynamics - Bone: dyn_hairbow_2_l!(ComponentPose=@n7, Alpha=1) -> (Pose) {
    angularDampingOverride=0.9999
    angularSpringConstant=700
    bAngularSpring=true
    bOverrideAngularDamping=true
    simulationSpace=RootRelative
    alphaScaleBiasClamp {scale=0.6}
    boundBone {boneName=dyn_hairbow_2_l}
    physicsBodyDefinitions [{
      boundBone {boneName=dyn_hairbow_2_l}
      boxExtents=<10 5 5>
      constraintSetup {angularLimitsMax=<15 15 15>, angularLimitsMin=<-15 -15 -15>, angularTarget=<1 0 0>}
      localJointOffset=<0.05 0 0>
     }]
    relativeSpaceBone {boneName=head}
   }
   n9 = Anim Dynamics - Bone: dyn_hairbow_r!(ComponentPose=@n8, Alpha=1) -> (Pose) {
    angularDampingOverride=0.9999
    angularSpringConstant=5000
    bAngularSpring=true
    bOverrideAngularDamping=true
    simulationSpace=RootRelative
    alphaScaleBiasClamp {scale=0.6}
    boundBone {boneName=dyn_hairbow_r}
    physicsBodyDefinitions [{
      boundBone {boneName=dyn_hairbow_r}
      boxExtents=<10 10 10>
      constraintSetup {angularLimitsMax=<15 15 15>, angularLimitsMin=<-15 -15 -15>, angularTarget=<1 0 0>}
      localJointOffset=<0.01 0 0>
     }]
    relativeSpaceBone {boneName=head}
   }
   n10 = Anim Dynamics - Bone: dyn_hairbow_2_r!(ComponentPose=@n9, Alpha=1) -> (Pose) {
    angularDampingOverride=0.999
    angularSpringConstant=700
    bAngularSpring=true
    bOverrideAngularDamping=true
    simulationSpace=RootRelative
    alphaScaleBiasClamp {scale=0.6}
    boundBone {boneName=dyn_hairbow_2_r}
    physicsBodyDefinitions [{
      boundBone {boneName=dyn_hairbow_2_r}
      boxExtents=<10 5 5>
      constraintSetup {angularLimitsMax=<15 15 15>, angularLimitsMin=<-15 -15 -15>, angularTarget=<1 0 0>}
      localJointOffset=<0.06 0 0>
     }]
    relativeSpaceBone {boneName=head}
   }
   n11 = Anim Dynamics - Bone: dyn_hairbow_1_r!(ComponentPose=@n10, Alpha=1) -> (Pose) {
    angularDampingOverride=0.999
    angularSpringConstant=700
    bAngularSpring=true
    bOverrideAngularDamping=true
    simulationSpace=RootRelative
    alphaScaleBiasClamp {scale=0.6}
    boundBone {boneName=dyn_hairbow_1_r}
    physicsBodyDefinitions [{
      boundBone {boneName=dyn_hairbow_1_r}
      boxExtents=<10 5 5>
      constraintSetup {angularLimitsMax=<15 15 15>, angularLimitsMin=<-15 -15 -15>, angularTarget=<1 0 0>}
      localJointOffset=<0.06 0 0>
     }]
    relativeSpaceBone {boneName=head}
   }
   n12 = Anim Dynamics - Bone: dyn_bow_f_l!(ComponentPose=@n11, Alpha=1) -> (Pose) {
    angularDampingOverride=0.9999
    angularSpringConstant=1500
    bAngularSpring=true
    bOverrideAngularDamping=true
    simulationSpace=RootRelative
    alphaScaleBiasClamp {scale=0.5}
    boundBone {boneName=dyn_bow_f_l}
    physicsBodyDefinitions [{
      boundBone {boneName=dyn_bow_f_l}
      boxExtents=<10 5 5>
      constraintSetup {angularLimitsMax=<15 15 15>, angularLimitsMin=<-15 -15 -15>, angularTarget=<1 0 0>}
      localJointOffset=<0.03 0 0>
     }]
    relativeSpaceBone {boneName=head}
   }
   n13 = Anim Dynamics - Bone: dyn_bow_f_r!(ComponentPose=@n12, Alpha=1) -> (Pose) {
    angularDampingOverride=0.9999
    angularSpringConstant=1500
    bAngularSpring=true
    bOverrideAngularDamping=true
    simulationSpace=RootRelative
    alphaScaleBiasClamp {scale=0.5}
    boundBone {boneName=dyn_bow_f_r}
    physicsBodyDefinitions [{
      boundBone {boneName=dyn_bow_f_r}
      boxExtents=<10 5 5>
      constraintSetup {angularLimitsMax=<15 15 15>, angularLimitsMin=<-15 -15 -15>, angularTarget=<1 0 0>}
      localJointOffset=<0.05 0 0>
     }]
    relativeSpaceBone {boneName=head}
   }
   Output Pose(Result=Component To Local(ComponentPose=@n13)) {layerGroup=None}
   n14 = Copy Pose From Mesh(SourceMeshComponent=Get Body_Mesh()) {bUseAttachedParent=true}
   n15 = Copy Pose From Mesh() {bUseAttachedParent=true}
   n16 = Anim Dynamics - Bone: dyn_hair_1_l!(
    ComponentPose=Local To Component(LocalPose=Layered blend per bone(BasePose=@n15, BlendPoses_0=@n14) {layerSetup [{branchFilters [{blendDepth=1, boneName=clavicle_l}, {blendDepth=1, boneName=clavicle_r}]}]})
    Alpha=1
   ) -> (Pose) {
    angularDampingOverride=0.9999
    angularSpringConstant=700
    bAngularSpring=true
    bLinearSpring=true
    bOverrideAngularDamping=true
    bOverrideLinearDamping=true
    bUseGravityOverride=true
    linearDampingOverride=0.9999
    linearSpringConstant=150
    simulationSpace=BoneRelative
    alphaScaleBiasClamp {scale=0.6}
    boundBone {boneName=dyn_hair_1_l}
    gravityOverride {z=-980}
    physicsBodyDefinitions [{
      boundBone {boneName=dyn_hair_1_l}
      boxExtents=<10 20 20>
      constraintSetup {angularLimitsMax=<0 10 10>, angularLimitsMin=<0 -10 -10>, angularTarget=<1 0 0>, linearAxesMax=<0 2 2>, linearAxesMin=<0 -2 -2>}
      localJointOffset=<0.01 0 0>
     }]
    relativeSpaceBone {boneName=neck_01}
   }
   n17 = Anim Dynamics - Bone: dyn_hair_1_r!(ComponentPose=@n16, Alpha=1) -> (Pose) {
    angularDampingOverride=0.9999
    angularSpringConstant=700
    bAngularSpring=true
    bLinearSpring=true
    bOverrideAngularDamping=true
    bOverrideLinearDamping=true
    bUseGravityOverride=true
    linearDampingOverride=0.9999
    linearSpringConstant=150
    simulationSpace=BoneRelative
    alphaScaleBiasClamp {scale=0.6}
    boundBone {boneName=dyn_hair_1_r}
    gravityOverride {z=-980}
    physicsBodyDefinitions [{
      boundBone {boneName=dyn_hair_1_r}
      boxExtents=<10 20 20>
      constraintSetup {angularLimitsMax=<0 10 10>, angularLimitsMin=<0 -10 -10>, angularTarget=<1 0 0>, linearAxesMax=<0 2 2>, linearAxesMin=<0 -2 -2>}
      localJointOffset=<0.01 0 0>
     }]
    relativeSpaceBone {boneName=neck_01}
   }
   n18 = Control Rig(Source=Component To Local(ComponentPose=@n17)) {controlRigAssetReference {blueprintRigClass=&/F_MED_PocketBobble_FaceAcc_CtrlRig.F_MED_PocketBobble_FaceAcc_CtrlRig_C}}
   Save cached pose 'PreRigidPose'(Pose=@n18) {cacheName=PreRigidPose}
  ]}
 EventGraph {nodes [
   Event BlueprintUpdateAnimation()
   > n1 = "Gravity Override State Function (Parameter Asset) (Nativized)"(
    CharacterPartAnimInstance=Self-Reference()
    GameTimeInSeconds=GetGameTimeInSeconds()
    Parameters=&/F_MED_PocketBobble_FaceAcc_Parameters
   )
   > Set Gravity Override(Gravity Override=@n1)
   > Set RBAN Settings Out(RBAN Settings Out=@n1.OutRigidBodyAnimNodeParameters)
   n2 = Event OnCharacterCustomizationLoaded()
   > Set Body_Mesh(Body_Mesh=Get Array Item with EFort Custom Part Type(Array=@n2.CustomMeshParts))
   n3 = CheckForBackbling()
   > Set isWearingPackBool(isWearingPackBool:bool=@n3) -> (Output_Get:bool)
   > Set WearingCape(WearingCape:double="To Float (Boolean)"(InBool=@n3.bIsBackblingACape)) -> (Output_Get:double)
  ]}
]
variables [
 Body_Mesh {type=SkeletalMeshComponent}
 Gravity Override {type=Vector}
 isWearingPackBool {type=boolean}
 RBAN Settings Out {type=FortRigidBodyAnimNodeParameters}
 WearingCape {type=number}
]
